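# Control Rig for the female medium CozyMechanic_Hen cosmetic body, correcting jumpsuit and strap bones during movement.
# After a shared DeformRig function, sequenced stages compute thigh pose-reader weights, Kalman-smoothed movement and pelvis-height crouch values.
# Parent constraints and weighted transform offsets use thigh-forward and crouch values, with extra fixes for emote, skydive and isFrEnd states.
# Thigh-side and forward/backward values are stored but unread; left/right filters are unconnected.
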ControlRigBlueprint /BRCosmetics/Characters/Player/Female/Medium/Bodies/F_MED_CozyMechanic_Hen/Meshes/F_MED_CozyMechanic_Hen_CtrlRig_B
graphs [
 RigVMFunctionLibrary {links [], nodes []}
 RigVMFunction_Sequence_2_ContainedGraph {nodes [
   n1 = RigVMFunctionEntryNode()
   n2 = &RigVMFunction_Sequence(ExecuteContext=@n1.ExecuteContext)
   n3 = &RigVMFunction_Sequence(ExecuteContext=@n2.B)
   n4 = &RigVMFunction_Sequence(ExecuteContext=@n3.B)
   n5 = &RigVMFunction_Sequence(ExecuteContext=@n4.B)
   n6 = &RigVMFunction_Sequence(ExecuteContext=@n5.B)
   RigVMFunctionReturnNode(A=@n2.A, B=@n3.A, C=@n4.A, D=@n5.A, E=@n6.A, F=@n6.B)
  ]}
 RigVMModel {nodes [
   RigUnit_BeginExecution = &RigUnit_BeginExecution()
   FN_DeformRig_FNC = RigVMFunctionReferenceNode(
    DeformRig_ Curve="(deformRig_l_upperarm_up_alpha_72_#1=False,deformRig_r_upperarm_up_alpha_73_#2=False,deformRig_l_upperarm_fr_alpha_74_#3=False,deformRig_r_upperarm_fr_alpha_75_CB51B8 ...(truncated)"
    DeformRig_Bool="(upperarm_1_#4=True,elbow_3_#5=True,wrist_7_#6=True,groin_8_#7=True,glute_17_#8=True,knee_9_40C408324CA261 ...(truncated)"
    DeformRig_Debug="(upperarm_debug=False,wrist_debug=False,lat_debug=False,groinGlute_debug=False,upperarm_debug_18_#4=False,wrist_debug_20_#6=False,groinGlute_debug_27_#7=False,lat_deb ...(truncated)"
    DeformRig_Value="(upperarm_up_18_#9=(Rotation=(X=0.0,Y=-0.0,Z=0.0,W=1.0),Translation=(X=0.0,Y=0.0,Z=8.0),Scale3D=(X=0.70,Y=1.0,Z=1.0)),upperarm_fwd_fr_19_#10=(Rotation=(X=0.0,Y=-0.0,Z=0.0,W=1.0),Translation=(X=0.0, ...(truncated)"
    ExecuteContext=@RigUnit_BeginExecution.ExecutePin
   )
   > RigVMFunction_Sequence_1 = RigVMAggregateNode()
   n1 = &RigUnit_SphericalPoseReader(
    ActiveRegionScaleFactors=(PositiveWidth=4,NegativeWidth=3,PositiveHeight=2.5,NegativeHeight=2)
    ActiveRegionSize=0.1
    Debug=(bDrawDebug=false,bDraw2D=False,bDrawLocalAxes=False,DebugScale=25,DebugSegments=20,DebugThickness=0.25)
    DriverAxis=(X=-1,Y=0,Z=0)
    DriverItem=(Type=Bone,Name="thigh_l")
    FalloffRegionScaleFactors=(PositiveWidth=1,NegativeWidth=3,PositiveHeight=1,NegativeHeight=0.5)
    FalloffSize=0.4
    FlipHeightScaling=False
    FlipWidthScaling=False
    OptionalParentItem=(Type=Bone,Name="None")
    RotationOffset=(X=100,Y=0,Z=-20)
    ExecutePin=@RigVMFunction_Sequence_1.A
   )
   > n2 = &RigUnit_SphericalPoseReader(
    ActiveRegionScaleFactors=(PositiveWidth=4,NegativeWidth=3,PositiveHeight=2.5,NegativeHeight=2)
    ActiveRegionSize=0.1
    Debug=(bDrawDebug=false,bDraw2D=False,bDrawLocalAxes=False,DebugScale=25,DebugSegments=20,DebugThickness=0.25)
    DriverAxis=(X=1,Y=0,Z=0)
    DriverItem=(Type=Bone,Name="thigh_r")
    FalloffRegionScaleFactors=(PositiveWidth=1,NegativeWidth=3.6,PositiveHeight=1,NegativeHeight=0.5)
    FalloffSize=0.4
    FlipHeightScaling=False
    FlipWidthScaling=true
    OptionalParentItem=(Type=Bone,Name="None")
    RotationOffset=(X=-80,Y=0,Z=-35.3405)
   )
   > n3 = &RigUnit_SphericalPoseReader(
    ActiveRegionScaleFactors=(PositiveWidth=3,NegativeWidth=2,PositiveHeight=2,NegativeHeight=2)
    ActiveRegionSize=0.1
    Debug=(bDrawDebug=false,bDraw2D=False,bDrawLocalAxes=False,DebugScale=25,DebugSegments=20,DebugThickness=0.25)
    DriverAxis=(X=-1,Y=0,Z=0)
    DriverItem=(Type=Bone,Name="thigh_l")
    FalloffRegionScaleFactors=(PositiveWidth=2,NegativeWidth=2,PositiveHeight=2,NegativeHeight=2)
    FalloffSize=0.25
    FlipHeightScaling=false
    FlipWidthScaling=false
    OptionalParentItem=(Type=Bone,Name="None")
    RotationOffset=(X=180,Y=-40,Z=0)
   )
   > n4 = &RigUnit_SphericalPoseReader(
    ActiveRegionScaleFactors=(PositiveWidth=3,NegativeWidth=2,PositiveHeight=2,NegativeHeight=2)
    ActiveRegionSize=0.1
    Debug=(bDrawDebug=false,bDraw2D=False,bDrawLocalAxes=False,DebugScale=25,DebugSegments=20,DebugThickness=0.25)
    DriverAxis=(X=1,Y=0,Z=0)
    DriverItem=(Type=Bone,Name="thigh_r")
    FalloffRegionScaleFactors=(PositiveWidth=2,NegativeWidth=2,PositiveHeight=2,NegativeHeight=2)
    FalloffSize=0.25
    FlipHeightScaling=False
    FlipWidthScaling=true
    OptionalParentItem=(Type=Bone,Name="None")
    RotationOffset=(X=0,Y=-40,Z=0)
   )
   > RigVMVariableNode(Value=@n1.OutputParam) {variable=Thigh_L_frv}
   > RigVMVariableNode(Value=@n2.OutputParam) {variable=Thigh_R_frv}
   > RigVMVariableNode(Value=@n3.OutputParam) {variable=Thigh_L_Side}
   > RigVMVariableNode(Value=@n4.OutputParam) {variable=Thigh_R_Side}
   n5 = RigVMVariableNode() {variable=Forward Backward Movement}
   KalmanFilter = &RigVMFunction_KalmanFloat(
    BufferSize=8
    Value=&RigVMFunction_MathDoubleRemap(SourceMaximum=1, SourceMinimum=0, TargetMaximum=1, TargetMinimum=0, Value=@n5.Value).Result
   ) {template="KalmanFilter::Execute(in Value,in BufferSize,out Result)"}
   > RigVMVariableNode(Value=@KalmanFilter.Result) {variable=Forward Movement}
   > RigVMVariableNode(Value=&RigVMFunction_KalmanFloat(
     BufferSize=8
     Value=&RigVMFunction_MathDoubleRemap(SourceMaximum=0, SourceMinimum=-1, TargetMaximum=0, TargetMinimum=1, Value=@n5.Value).Result
    ) {template="KalmanFilter::Execute(in Value,in BufferSize,out Result)"}.Result) {variable=Backward Movement}
   n6 = &RigUnit_GetRelativeTransformForItem(Child=(Type=Bone,Name="pelvis"), Parent=(Type=Bone,Name="root"), bChildInitial=False, bParentInitial=False)
   > RigVMVariableNode(Value=&RigVMFunction_MathDoubleMul(
     B=&RigVMFunction_MathDoubleRemap(SourceMaximum=67, SourceMinimum=90, TargetMaximum=1, TargetMinimum=0, Value=@n6.RelativeTransform.Translation.Z).Result
     A=&RigVMFunction_MathDoubleRemap(SourceMaximum=69, SourceMinimum=50, TargetMaximum=1, TargetMinimum=0, Value=@n6.RelativeTransform.Translation.Z).Result
    ).Result) {variable=CrouchWalk}
   > RigVMVariableNode(Value=&RigVMFunction_MathDoubleRemap(
     SourceMaximum=40
     SourceMinimum=85
     TargetMaximum=1
     TargetMinimum=0
     Value=&RigUnit_GetRelativeTransformForItem(Child=(Type=Bone,Name="pelvis"), Parent=(Type=Bone,Name="root"), bChildInitial=False, bParentInitial=False).RelativeTransform.Translation.Z
    ).Result) {variable=Crouch}
   &RigUnit_ParentConstraint(
    AdvancedSettings=(InterpolationType=Average,RotationOrderForFilter=XZY)
    Child=(Type=Bone,Name="jumpsuit_Base_l")
    Parents="((Item=(Type=Bone,Name=\"pelvis\"),Weight=0.1361),(Item=(Type=Bone,Name=\"thigh_twist_01_l\"),Weight=0.1945),(Item=(Type=Bone,Name=\"thigh_twist_01_r\"),Weight=0.0173),(Item=(Type=Bone,Name=\"thigh_l\"),Weight=0.0637),(Item=(Type=Bone,Name=\"deform_groin_l\"),Weight ...(truncated)"
    Weight=1
    ExecutePin=@RigVMFunction_Sequence_1.B
   )
   > &RigUnit_ParentConstraint(
    AdvancedSettings=(InterpolationType=Average,RotationOrderForFilter=XZY)
    Child=(Type=Bone,Name="jumpsuit_Base_r")
    Parents=((Item=(Type=Bone,Name="pelvis"),Weight=0.159),(Item=(Type=Bone,Name="thigh_twist_01_r"),Weight=0.0183),(Item=(Type=Bone,Name="thigh_r"),Weight=0.0914),(Item=(Type=Bone,Name="deform_groin_r"),Weight=0.4584))
    Weight=1
   )
   > &RigUnit_ModifyTransforms(
    ItemToModify=((Item=(Type=Bone,Name="jumpsuit_Base_l"),Transform=(Rotation=(X=0,Y=-0.1736,Z=0,W=0.9848),Translation=(X=0,Y=0,Z=1),Scale3D=(X=1,Y=1,Z=1))))
    WeightMinimum=0
    Weight=RigVMVariableNode() {variable=Thigh_L_frv}.Value
   )
   > &RigUnit_ModifyTransforms(
    ItemToModify=((Item=(Type=Bone,Name="jumpsuit_Base_r"),Transform=(Rotation=(X=0,Y=-0.1736,Z=0,W=0.9848),Translation=(X=1,Y=-1,Z=0),Scale3D=(X=1,Y=1,Z=1))))
    WeightMinimum=0
    Weight=RigVMVariableNode() {variable=Thigh_R_frv}.Value
   )
   > &RigUnit_ModifyTransforms(
    ItemToModify="((Item=(Type=Bone,Name=\"jumpsuit_Base_l\"),Transform=(Rotation=(X=0,Y=0,Z=0.2302,W=0.9731),Translation=(X=0,Y=0,Z=1.4),Scale3D=(X=1,Y=1,Z=1))),(Item=(Type=Bone,Name=\"Strap_corr_base_l\"),Transform=(Rotation=(X=0,Y=0,Z=0,W=1),Translation=(X=0,Y=0,Z=0),Scale3D ...(truncated)"
    WeightMinimum=0
    Weight=RigVMVariableNode() {variable=Crouch}.Value
   )
   > &RigUnit_ModifyTransforms(
    ItemToModify="((Item=(Type=Bone,Name=\"Strap_Back_02_Base_l\"),Transform=(Rotation=(X=0,Y=-0.0955,Z=0,W=0.9954),Translation=(X=0,Y=0,Z=3),Scale3D=(X=1,Y=1,Z=1))),(Item=(Type=Bone,Name=\"Strap_corr_r\"),Transform=(Rotation=(X=0,Y=0,Z=0,W=1),Translation=(X=-3,Y=0,Z=0),Scale3D ...(truncated)"
    WeightMinimum=0
    Weight=RigVMVariableNode() {variable=CrouchWalk}.Value
   )
   > &RigUnit_ModifyTransforms(
    ItemToModify=((Item=(Type=Bone,Name="Strap_corr_l"),Transform=(Rotation=(X=0,Y=0,Z=0,W=1),Translation=(X=-5,Y=-2,Z=0),Scale3D=(X=1,Y=1,Z=1))))
    WeightMinimum=0
    Weight=RigVMVariableNode() {variable=Thigh_L_frv}.Value
   )
   &RigUnit_ModifyTransforms(
    ItemToModify="((Item=(Type=Bone,Name=\"Strap_corr_base_r\"),Transform=(Rotation=(X=0,Y=0,Z=0.0436,W=0.999),Translation=(X=0,Y=0,Z=0),Scale3D=(X=1,Y=1,Z=1))),(Item=(Type=Bone,Name=\"Strap_corr_base_l\"),Transform=(Rotation=(X=0,Y=0,Z=0.0436,W=0.999),Translation=(X=0,Y=0,Z=0) ...(truncated)"
    Weight=1
    WeightMinimum=0
    ExecutePin=@RigVMFunction_Sequence_1.C
   )
   n7 = &RigVMFunction_ControlFlowBranch(ExecuteContext=@RigVMFunction_Sequence_1.D, Condition=RigVMVariableNode() {variable=IsEmote}.Value)
   n8 = &RigUnit_SphericalPoseReader(
    ActiveRegionScaleFactors=(PositiveWidth=0.8,NegativeWidth=0,PositiveHeight=0.2,NegativeHeight=0.5)
    ActiveRegionSize=0.3
    Debug=(bDrawDebug=false,bDraw2D=False,bDrawLocalAxes=False,DebugScale=25,DebugSegments=20,DebugThickness=0.25)
    DriverAxis=(X=1,Y=0,Z=0)
    DriverItem=(Type=Bone,Name="thigh_r")
    FalloffRegionScaleFactors=(PositiveWidth=0.5,NegativeWidth=0.5,PositiveHeight=0.5,NegativeHeight=0.5)
    FalloffSize=0.6
    FlipHeightScaling=False
    FlipWidthScaling=False
    OptionalParentItem=(Type=Bone,Name="pelvis")
    RotationOffset=(X=-90,Y=0,Z=-60)
    ExecutePin=@n7.True
   )
   > &RigUnit_ModifyTransforms(
    ItemToModify="((Item=(Type=Bone,Name=\"jumpsuit_Base_r\"),Transform=(Rotation=(X=0,Y=0.0872,Z=0,W=-0.9962),Translation=(X=-0.6282,Y=0,Z=1),Scale3D=(X=1,Y=1,Z=1))),(Item=(Type=Bone,Name=\"Strap_corr_r\"),Transform=(Rotation=(X=0,Y=0,Z=0,W=1),Translation=(X=-5,Y=0,Z=0),Scale3 ...(truncated)"
    WeightMinimum=0
    Weight=@n8.OutputParam
   )
   RigVMFunction_ControlFlowBranch = &RigVMFunction_ControlFlowBranch(ExecuteContext=@RigVMFunction_Sequence_1.E, Condition=RigVMVariableNode() {variable=isSkyDive}.Value)
   &RigUnit_ModifyTransforms(
    ItemToModify="((Item=(Type=Bone,Name=\"Strap_corr_l\"),Transform=(Rotation=(X=0,Y=0,Z=0,W=-1),Translation=(X=-3.3804,Y=0.1294,Z=-4.1614),Scale3D=(X=1,Y=1,Z=1))),(Item=(Type=Bone,Name=\"Strap_corr_r\"),Transform=(Rotation=(X=0,Y=0,Z=0,W=1),Translation=(X=-3.2074,Y=3.5347,Z=0 ...(truncated)"
    Weight=1
    WeightMinimum=0
    ExecutePin=@RigVMFunction_ControlFlowBranch.True
   )
   > &RigUnit_ModifyTransforms(
    ItemToModify="((Item=(Type=Bone,Name=\"upperarm_twist_02_r\"),Transform=(Rotation=(X=0,Y=0,Z=0,W=1),Translation=(X=0,Y=0,Z=0),Scale3D=(X=2,Y=2.1,Z=2))),(Item=(Type=Bone,Name=\"upperarm_twist_01_r\"),Transform=(Rotation=(X=0,Y=0,Z=0,W=1),Translation=(X=0,Y=0,Z=0),Scale3D=(X= ...(truncated)"
    Weight=1
    WeightMinimum=0
   )
   > &RigUnit_ModifyTransforms(
    ItemToModify="((Item=(Type=Bone,Name=\"upperarm_twist_02_l\"),Transform=(Rotation=(X=0,Y=0,Z=0,W=1),Translation=(X=0,Y=0,Z=0),Scale3D=(X=2,Y=2.1,Z=2))),(Item=(Type=Bone,Name=\"upperarm_twist_01_l\"),Transform=(Rotation=(X=0,Y=0,Z=0,W=1),Translation=(X=0,Y=0,Z=0),Scale3D=(X= ...(truncated)"
    Weight=1
    WeightMinimum=0
   )
   n9 = &RigVMFunction_ControlFlowBranch(ExecuteContext=@RigVMFunction_Sequence_1.F, Condition=RigVMVariableNode() {variable=isFrEnd}.Value)
   &RigUnit_ModifyTransforms(
    ItemToModify=((Item=(Type=Bone,Name="Strap_corr_r"),Transform=(Rotation=(X=0,Y=0,Z=0,W=1),Translation=(X=0,Y=1,Z=0),Scale3D=(X=1,Y=1,Z=1))))
    Weight=1
    WeightMinimum=0
    ExecutePin=@n9.True
   )
   n10 = RigVMVariableNode() {variable=Left Right Movement}
   &RigVMFunction_KalmanFloat(
    BufferSize=8
    Value=&RigVMFunction_MathDoubleRemap(SourceMaximum=0, SourceMinimum=-1, TargetMaximum=0, TargetMinimum=1, Value=@n10.Value).Result
   ) {template="KalmanFilter::Execute(in Value,in BufferSize,out Result)"}
   &RigVMFunction_KalmanFloat(
    BufferSize=8
    Value=&RigVMFunction_MathDoubleRemap(SourceMaximum=1, SourceMinimum=0, TargetMaximum=1, TargetMinimum=0, Value=@n10.Value).Result
   ) {template="KalmanFilter::Execute(in Value,in BufferSize,out Result)"}
  ]}
]
hierarchy [
 root {loc=<0 0 0>, rot=<0 0 0 1>}
 pelvis {parent=root, loc=<0 -0.5301 98.0571>, rot=<0.0127 0.707 -0.0127 -0.707>}
 spine_01 {parent=pelvis, loc=<6.1208 -0.9328 0>, rot=<0 0 0.034 -0.9994>}
 spine_02 {parent=spine_01, loc=<6.5997 -0.2095 0>, rot=<0 0 0.0238 -0.9997>}
 spine_03 {parent=spine_02, loc=<6.4914 -0.1237 0>, rot=<0 0 0.0149 -0.9999>}
 spine_04 {parent=spine_03, loc=<6.9598 0.0794 0>, rot=<0 0 -0.0132 -0.9999>}
 spine_05 {parent=spine_04, loc=<6.9987 0 0>, rot=<0 0 -0.0015 -1>}
 neck_01 {parent=spine_05, loc=<11.3138 0.1529 0>, rot=<0 0 -0.2003 -0.9797>}
 neck_02 {parent=neck_01, loc=<4.5627 0 0>, rot=<0 0 0.0504 -0.9987>}
 head {parent=neck_02, loc=<5.3914 0 0>, rot=<0 0 0.115 -0.9934>}
 faceAttach {parent=head, loc=<-4.2319 -0.6469 -0.0667>, rot=<0.0082 -0.7071 -0.0082 -0.7071>}
 C_brow_mid {parent=faceAttach, loc=<-0.0892 9.1228 9.262>, rot=<0 -0.0246 0.0049 -0.9997>}
 L_brow_mid {parent=faceAttach, loc=<3.3313 8.5867 10.043>, rot=<-0.0011 -0.0118 0.2553 -0.9668>}
 L_brow_outer {parent=faceAttach, loc=<5.6274 6.2932 10.3805>, rot=<0.0847 0.1027 0.4187 -0.8983>}
 L_cheek_inner {parent=faceAttach, loc=<4.7138 7.0294 3.7516>, rot=<-0.0288 -0.1637 0.331 -0.9289>}
 C_lip_upper_mid {parent=faceAttach, loc=<0.0055 9.4234 1.91>, rot=<0.2225 0.0007 -0.0002 -0.9749>}
 L_lip_upper_outer {parent=faceAttach, loc=<1.361 8.2931 2.0418>, rot=<0.0114 -0.1415 0.2878 -0.9471>}
 L_lip_corner {parent=faceAttach, loc=<1.8732 8.206 1.9638>, rot=<0.0036 0.0073 -0.0147 -0.9999>}
 C_nose_bridge {parent=faceAttach, loc=<-0.1039 10.269 5.8106>, rot=<-0.2218 -0.017 -0.0048 -0.9749>}
 R_brow_mid {parent=faceAttach, loc=<-3.4643 8.5866 10.0483>, rot=<-0.0113 0.0905 -0.255 -0.9626>}
 R_brow_outer {parent=faceAttach, loc=<-5.7747 6.2627 10.378>, rot=<0.0422 -0.0492 -0.4319 -0.8996>}
 R_cheek_inner {parent=faceAttach, loc=<-4.838 7.0359 3.7426>, rot=<-0.0272 0.1596 -0.3314 -0.9295>}
 R_lip_corner {parent=faceAttach, loc=<-2.0528 8.1555 1.9497>, rot=<0.0031 -0.0078 0.0146 -0.9999>}
 C_jaw {parent=faceAttach, loc=<-0.0824 2.7035 1.1048>, rot=<0 0 0 1>}
 L_lip_lower_outer {parent=C_jaw, loc=<1.5282 5.6399 0.8237>, rot=<0.0191 0.078 0.3307 -0.9403>}
 C_lip_lower_mid {parent=C_jaw, loc=<0.0996 6.5406 0.9864>, rot=<0 -0.0178 -0.0038 -0.9998>}
 R_lip_lower_outer {parent=C_jaw, loc=<-1.4995 5.8896 0.8329>, rot=<0.0143 -0.0516 -0.311 -0.9489>}
 teeth_lower {parent=C_jaw, loc=<0.0633 3.0611 -0.7691>, rot=<0 0 0 1>}
 tongue {parent=C_jaw, loc=<0.0545 1.3613 -0.9565>, rot=<-0.1568 0 0 -0.9876>}
 R_eye {parent=faceAttach, loc=<-3.1701 6.3747 8.2895>, rot=<0 0 0.9998 -0.0187>}
 R_eye_lid_upper_mid {parent=R_eye, loc=<-0.002 -1.0617 0.1233>, rot=<-0.0003 0 0.0007 -1>}
 R_eye_lid_lower_mid {parent=R_eye, loc=<0.0193 -0.795 -0.6819>, rot=<-0.0003 0 0.0007 -1>}
 L_eye {parent=faceAttach, loc=<3.0549 6.3747 8.2895>, rot=<0 0 0.9984 0.0568>}
 L_eye_lid_upper_mid {parent=L_eye, loc=<0.0224 -1.051 0.1034>, rot=<0 0 0 1>}
 L_eye_lid_lower_mid {parent=L_eye, loc=<0.0198 -0.7759 -0.719>, rot=<0 0 0 1>}
 teeth_upper {parent=faceAttach, loc=<-0.023 6.4156 1.7174>, rot=<0 0 0 1>}
 R_lip_upper_outer {parent=faceAttach, loc=<-1.5073 8.3248 2.0867>, rot=<0.0121 0.1395 -0.2871 -0.9476>}
 zipper_base {parent=neck_01, loc=<2.9608 4.798 0>, rot=<-0.1608 -0.987 0 0>}
 zipper_dyn {parent=zipper_base, loc=<0 0.2731 0>, rot=<0 0 0 1>}
 clavicle_l {parent=spine_05, loc=<5.4831 -0.1053 -3.2946>, rot=<0.0465 -0.7289 -0.009 -0.6829>}
 upperarm_l {parent=clavicle_l, loc=<9.8658 -1.2136 -0.4105>, rot=<0.0164 -0.3245 -0.0269 -0.9454>}
 lowerarm_l {parent=upperarm_l, loc=<26.465 -0.8345 0.259>, rot=<0.0036 0.0161 -0.2298 -0.9731>}
 hand_l {parent=lowerarm_l, loc=<22.7355 -0.3203 -0.6105>, rot=<-0.6784 -0.0223 0.0735 -0.7307>}
 thumb_01_l {parent=hand_l, loc=<1.3118 -0.3191 -2.459>, rot=<0.6546 -0.4042 0.0664 -0.6354>}
 thumb_02_l {parent=thumb_01_l, loc=<5.9336 -0.0448 0.0401>, rot=<-0.0037 0.0213 0.298 -0.9543>}
 thumb_03_l {parent=thumb_02_l, loc=<2.2223 0.3682 -0.0907>, rot=<0.0118 -0.013 0.0066 -0.9998>}
 index_metacarpal_l {parent=hand_l, loc=<1.6535 0.0183 -2.96>, rot=<0.0158 -0.0124 -0.0173 -0.9996>}
 index_01_l {parent=index_metacarpal_l, loc=<8.3008 0.0357 -0.0524>, rot=<-0.0044 0.0485 0.0924 -0.9945>}
 index_02_l {parent=index_01_l, loc=<3.3587 -0.3004 0.0452>, rot=<-0.002 0.0027 0.3225 -0.9466>}
 index_03_l {parent=index_02_l, loc=<2.5578 -0.123 -0.0252>, rot=<-0.002 0.0067 0.2558 -0.9667>}
 middle_metacarpal_l {parent=hand_l, loc=<2.1856 -0.1725 -0.5984>, rot=<0.006 0.0168 -0.0329 -0.9993>}
 middle_01_l {parent=middle_metacarpal_l, loc=<7.3275 -0.2447 0.0129>, rot=<-0.0208 0.0042 0.0858 -0.9961>}
 middle_02_l {parent=middle_01_l, loc=<4.0123 0.1401 0.1175>, rot=<-0.0218 0.0258 0.3512 -0.9357>}
 middle_03_l {parent=middle_02_l, loc=<2.6077 0.0534 0.1373>, rot=<0.0082 0.0188 0.1708 -0.9851>}
 ring_metacarpal_l {parent=hand_l, loc=<2.2074 -0.3625 1.3638>, rot=<-0.0284 0.0722 -0.0458 -0.9959>}
 ring_01_l {parent=ring_metacarpal_l, loc=<7.0295 -0.6197 -0.1446>, rot=<-0.0489 -0.0252 0.1892 -0.9804>}
 ring_02_l {parent=ring_01_l, loc=<3.6174 0.4647 -0.1027>, rot=<0 0 0.2806 -0.9598>}
 ring_03_l {parent=ring_02_l, loc=<2.4152 0.114 0.0199>, rot=<0 0 0.1189 -0.9929>}
 pinky_metacarpal_l {parent=hand_l, loc=<1.7616 -0.9753 3.3814>, rot=<-0.0397 0.0814 -0.069 -0.9935>}
 pinky_01_l {parent=pinky_metacarpal_l, loc=<7.1709 -0.4831 -0.1094>, rot=<-0.0654 -0.0035 0.2242 -0.9723>}
 pinky_02_l {parent=pinky_01_l, loc=<2.8274 0.0893 -0.1043>, rot=<0.0037 0.0011 0.2854 -0.9584>}
 pinky_03_l {parent=pinky_02_l, loc=<2.2046 0.0307 -0.0078>, rot=<-0.0002 0.0098 0.1445 -0.9895>}
 deform_wrist_dn_base_l {parent=hand_l, loc=<0 -1.5326 0>, rot=<0 0 0 -1>}
 deform_wrist_dn_l {parent=deform_wrist_dn_base_l, loc=<0 0 0>, rot=<0 0 0 1>}
 deform_wrist_up_base_l {parent=hand_l, loc=<0 2.3009 0>, rot=<0 0 0 -1>}
 deform_wrist_up_l {parent=deform_wrist_up_base_l, loc=<0 0 0>, rot=<0 0 0 1>}
 lowerarm_twist_02_l {parent=lowerarm_l, loc=<10.6707 0 0>, rot=<0 0 0 1>}
 lowerarm_twist_01_l {parent=lowerarm_l, loc=<18.316 0 0>, rot=<0 0 0 1>}
 deform_elbow_in_base_l {parent=lowerarm_l, loc=<0.1834 3.6062 -0.2572>, rot=<0 0 0 -1>}
 deform_elbow_in_l {parent=deform_elbow_in_base_l, loc=<0 0 0>, rot=<0 0 0 1>}
 deform_elbow_out_base_l {parent=lowerarm_l, loc=<0 -4 0>, rot=<0 0 0 1>}
 deform_elbow_out_l {parent=deform_elbow_out_base_l, loc=<0 0 0>, rot=<0 0 0 1>}
 upperarm_twist_01_l {parent=upperarm_l, loc=<7.2135 -0.1134 0>, rot=<0 0 0 1>}
 deform_upperarm_bk_base_l {parent=upperarm_twist_01_l, loc=<-6.4053 -4.3746 0.314>, rot=<0 0 0 -1>}
 deform_upperarm_bk_l {parent=deform_upperarm_bk_base_l, loc=<0 0 0>, rot=<0 0 0 1>}
 deform_upperarm_fr_base_l {parent=upperarm_twist_01_l, loc=<-5.6312 5.092 0.3354>, rot=<0 0 0 -1>}
 deform_upperarm_fr_l {parent=deform_upperarm_fr_base_l, loc=<0 0 0>, rot=<0 0 0 1>}
 deform_upperarm_up_base_l {parent=upperarm_twist_01_l, loc=<-8.891 0.4825 3.9517>, rot=<0 0 0 -1>}
 deform_upperarm_up_l {parent=deform_upperarm_up_base_l, loc=<0 0 0>, rot=<0 0 0 1>}
 upperarm_twist_02_l {parent=upperarm_l, loc=<19.7998 -0.1134 0>, rot=<0 0 0 1>}
 clavicle_r {parent=spine_05, loc=<5.4831 -0.1053 3.2946>, rot=<0.7289 0.0465 -0.6829 0.009>}
 upperarm_r {parent=clavicle_r, loc=<-9.8658 1.2136 0.4105>, rot=<0.0164 -0.3245 -0.0269 -0.9454>}
 lowerarm_r {parent=upperarm_r, loc=<-26.465 0.8345 -0.259>, rot=<0.0036 0.0161 -0.2298 -0.9731>}
 hand_r {parent=lowerarm_r, loc=<-22.7355 0.3203 0.6106>, rot=<-0.6784 -0.0223 0.0735 -0.7307>}
 thumb_01_r {parent=hand_r, loc=<-1.3118 0.3191 2.459>, rot=<0.6546 -0.4042 0.0664 -0.6354>}
 thumb_02_r {parent=thumb_01_r, loc=<-5.9336 0.0448 -0.0401>, rot=<-0.0037 0.0213 0.298 -0.9543>}
 thumb_03_r {parent=thumb_02_r, loc=<-2.2223 -0.3682 0.0907>, rot=<0.0118 -0.013 0.0066 -0.9998>}
 index_metacarpal_r {parent=hand_r, loc=<-1.6535 -0.0183 2.96>, rot=<0.0158 -0.0124 -0.0173 -0.9996>}
 index_01_r {parent=index_metacarpal_r, loc=<-8.3008 -0.0357 0.0524>, rot=<-0.0044 0.0485 0.0924 -0.9945>}
 index_02_r {parent=index_01_r, loc=<-3.3587 0.3004 -0.0452>, rot=<-0.002 0.0027 0.3225 -0.9466>}
 index_03_r {parent=index_02_r, loc=<-2.5578 0.123 0.0252>, rot=<-0.002 0.0067 0.2558 -0.9667>}
 middle_metacarpal_r {parent=hand_r, loc=<-2.1856 0.1725 0.5984>, rot=<0.006 0.0168 -0.0329 -0.9993>}
 middle_01_r {parent=middle_metacarpal_r, loc=<-7.3275 0.2447 -0.0129>, rot=<-0.0208 0.0042 0.0858 -0.9961>}
 middle_02_r {parent=middle_01_r, loc=<-4.0123 -0.1401 -0.1175>, rot=<-0.0218 0.0258 0.3512 -0.9357>}
 middle_03_r {parent=middle_02_r, loc=<-2.6077 -0.0534 -0.1373>, rot=<0.0082 0.0188 0.1708 -0.9851>}
 ring_metacarpal_r {parent=hand_r, loc=<-2.2074 0.3625 -1.3638>, rot=<-0.0284 0.0722 -0.0458 -0.9959>}
 ring_01_r {parent=ring_metacarpal_r, loc=<-7.0295 0.6197 0.1446>, rot=<-0.0489 -0.0252 0.1892 -0.9804>}
 ring_02_r {parent=ring_01_r, loc=<-3.6174 -0.4647 0.1027>, rot=<0 0 0.2806 -0.9598>}
 ring_03_r {parent=ring_02_r, loc=<-2.4152 -0.114 -0.0199>, rot=<0 0 0.1189 -0.9929>}
 pinky_metacarpal_r {parent=hand_r, loc=<-1.7616 0.9753 -3.3814>, rot=<-0.0397 0.0814 -0.069 -0.9935>}
 pinky_01_r {parent=pinky_metacarpal_r, loc=<-7.1709 0.4831 0.1094>, rot=<-0.0654 -0.0035 0.2242 -0.9723>}
 pinky_02_r {parent=pinky_01_r, loc=<-2.8274 -0.0893 0.1043>, rot=<0.0037 0.0011 0.2854 -0.9584>}
 pinky_03_r {parent=pinky_02_r, loc=<-2.2046 -0.0307 0.0078>, rot=<-0.0002 0.0098 0.1445 -0.9895>}
 deform_wrist_dn_base_r {parent=hand_r, loc=<-0.0001 1.5328 0>, rot=<0 0 0 -1>}
 deform_wrist_dn_r {parent=deform_wrist_dn_base_r, loc=<0 0 0>, rot=<0 0 0 1>}
 deform_wrist_up_base_r {parent=hand_r, loc=<-0.0002 -2.3007 -0.0001>, rot=<0 0 0 -1>}
 deform_wrist_up_r {parent=deform_wrist_up_base_r, loc=<0 0 0>, rot=<0 0 0 1>}
 lowerarm_twist_02_r {parent=lowerarm_r, loc=<-10.6707 0 0>, rot=<0 0 0 -1>}
 lowerarm_twist_01_r {parent=lowerarm_r, loc=<-18.316 0 0>, rot=<0 0 0 -1>}
 deform_elbow_in_base_r {parent=lowerarm_r, loc=<-0.1834 -3.6062 0.2572>, rot=<0 0 0 -1>}
 deform_elbow_in_r {parent=deform_elbow_in_base_r, loc=<0 0 0>, rot=<0 0 0 1>}
 deform_elbow_out_base_r {parent=lowerarm_r, loc=<0 4 -0.0001>, rot=<0 0 0 -1>}
 deform_elbow_out_r {parent=deform_elbow_out_base_r, loc=<0 0 0>, rot=<0 0 0 1>}
 upperarm_twist_01_r {parent=upperarm_r, loc=<-7.2135 0.1134 0>, rot=<0 0 0 -1>}
 deform_upperarm_bk_base_r {parent=upperarm_twist_01_r, loc=<6.4053 4.3746 -0.314>, rot=<0 0 0 -1>}
 deform_upperarm_bk_r {parent=deform_upperarm_bk_base_r, loc=<0 0 0>, rot=<0 0 0 1>}
 deform_upperarm_fr_base_r {parent=upperarm_twist_01_r, loc=<5.6312 -5.092 -0.3354>, rot=<0 0 0 -1>}
 deform_upperarm_fr_r {parent=deform_upperarm_fr_base_r, loc=<0 0 0>, rot=<0 0 0 1>}
 deform_upperarm_up_base_r {parent=upperarm_twist_01_r, loc=<8.891 -0.4825 -3.9517>, rot=<0 0 0 -1>}
 deform_upperarm_up_r {parent=deform_upperarm_up_base_r, loc=<0 0 0>, rot=<0 0 0 1>}
 upperarm_twist_02_r {parent=upperarm_r, loc=<-19.7998 0.1134 0>, rot=<0 0 0 -1>}
 attach_backpack {parent=spine_05, loc=<-0.1204 -13.4677 -0.0006>, rot=<-0.0093 -0.7071 0.0095 -0.707>}
 Backpack_BR {parent=attach_backpack, type=Null, loc=<0 4.9406 0>, rot=<0 0 0 1>}
 attach_cape {parent=spine_05, loc=<9.0182 -11.7112 -0.0014>, rot=<0.0616 -0.7044 -0.0614 -0.7044>}
 cape {parent=attach_cape, type=Null, loc=<0 5.0069 0>, rot=<0 0 0 1>}
 deform_pec_aim_base_l {parent=spine_04, loc=<3.2456 9.3309 -1.2237>, rot=<0.0603 -0.4033 0.3318 -0.8507>}
 deform_pec_aim_l {parent=deform_pec_aim_base_l, loc=<0 0 0>, rot=<0 0 0 1>}
 deform_pec_base_l {parent=deform_pec_aim_l, loc=<9.3665 0 0>, rot=<0 0 0 1>}
 deform_pec_l {parent=deform_pec_base_l, loc=<0 0 0>, rot=<0 0 0 1>}
 deform_pec_aim_base_r {parent=spine_04, loc=<3.2452 9.3309 1.2237>, rot=<-0.4033 -0.0603 0.8507 0.3318>}
 deform_pec_aim_r {parent=deform_pec_aim_base_r, loc=<0 0 0>, rot=<0 0 0 1>}
 deform_pec_base_r {parent=deform_pec_aim_r, loc=<-9.3667 -0.0003 -0.0004>, rot=<0 0 0 1>}
 deform_pec_r {parent=deform_pec_base_r, loc=<0 0 0>, rot=<0 0 0 1>}
 thigh_l {parent=pelvis, loc=<-1.0855 -0.2109 -8.7053>, rot=<0.0009 0.0355 0.0155 -0.9993>}
 calf_l {parent=thigh_l, loc=<-42.5284 -0.0019 -0.7297>, rot=<0.0006 -0.0161 -0.0391 -0.9991>}
 foot_l {parent=calf_l, loc=<-46.7401 -2.5014 -0.592>, rot=<0.0427 -0.0238 0.0353 -0.9982>}
 ball_l {parent=foot_l, loc=<-3.5658 10.8066 -0.1544>, rot=<0.0001 -0.0001 -0.7063 -0.7079>}
 Foot_corr_l {parent=foot_l, loc=<1.7451 -5.4184 0.1102>, rot=<-0.0357 -0.7045 -0.0284 -0.7082>}
 calf_twist_02_l {parent=calf_l, loc=<-19.1546 -1.7943 -0.2638>, rot=<0.0237 0.0095 -0.0035 -0.9997>}
 calf_twist_01_l {parent=calf_l, loc=<-35.8483 -1.9197 -0.5794>, rot=<0.0237 0.0095 -0.0035 -0.9997>}
 deform_knee_in_base_l {parent=calf_l, loc=<0.8031 -5.0231 0.2456>, rot=<0 0 0 -1>}
 deform_knee_in_l {parent=deform_knee_in_base_l, loc=<0 0 0>, rot=<0 0 0 1>}
 deform_knee_out_base_l {parent=calf_l, loc=<0 4 0>, rot=<0 0 0 -1>}
 deform_knee_out_l {parent=deform_knee_out_base_l, loc=<0 0 0>, rot=<0 0 0 1>}
 thigh_twist_01_l {parent=thigh_l, loc=<-27.952 0 0>, rot=<0.0013 0 0 -1>}
 deform_groin_base_l {parent=thigh_l, loc=<0.4801 5.6776 -0.2486>, rot=<0 0 0 -1>}
 deform_groin_l {parent=deform_groin_base_l, loc=<0 0 0>, rot=<0 0 0 1>}
 Strap_corr_base_l {parent=thigh_l, loc=<0 0 0>, rot=<-0.0308 -0.7064 0.7064 -0.0308>}
 Strap_corr_l {parent=Strap_corr_base_l, loc=<24.7801 7.6454 -0.644>, rot=<0 0 0 1>}
 jumpsuit_Base_l {parent=thigh_l, loc=<-6.6721 9.1908 2.1005>, rot=<0.2036 -0.6922 -0.6764 -0.1479>}
 Strap_01_dyn_l {parent=jumpsuit_Base_l, loc=<0 0 0>, rot=<0 0 0 1>}
 Strap_02_dyn_l {parent=Strap_01_dyn_l, loc=<6.4184 0 0>, rot=<0 0 0 1>}
 Strap_03_dyn_l {parent=Strap_02_dyn_l, loc=<5.439 0 0>, rot=<0 -0.1325 0 -0.9912>}
 Strap_04_dyn_l {parent=Strap_03_dyn_l, loc=<6.411 0 0>, rot=<-0.1879 -0.2849 0.0571 -0.9382>}
 Strap_05_dyn_l {parent=Strap_04_dyn_l, loc=<5.8985 0 0>, rot=<0 0 0 1>}
 thigh_r {parent=pelvis, loc=<-1.0855 -0.2109 8.7053>, rot=<0.0355 -0.0009 0.9993 0.0155>}
 calf_r {parent=thigh_r, loc=<42.5284 0.0019 0.7297>, rot=<0.0006 -0.0161 -0.0391 -0.9991>}
 foot_r {parent=calf_r, loc=<46.7401 2.5014 0.592>, rot=<0.0427 -0.0238 0.0353 -0.9982>}
 ball_r {parent=foot_r, loc=<3.5658 -10.8066 0.1544>, rot=<0.0001 -0.0001 -0.7063 -0.7079>}
 Foot_corr_r {parent=foot_r, loc=<-1.7451 5.4184 -0.1102>, rot=<0.7082 0.0284 -0.7045 -0.0357>}
 calf_twist_02_r {parent=calf_r, loc=<19.1546 1.7943 0.2638>, rot=<0.0237 0.0095 -0.0035 -0.9997>}
 calf_twist_01_r {parent=calf_r, loc=<35.8483 1.9197 0.5794>, rot=<0.0237 0.0095 -0.0035 -0.9997>}
 deform_knee_in_base_r {parent=calf_r, loc=<-0.8031 5.0231 -0.2457>, rot=<0 0 0 -1>}
 deform_knee_in_r {parent=deform_knee_in_base_r, loc=<0 0 0>, rot=<0 0 0 1>}
 deform_knee_out_base_r {parent=calf_r, loc=<0 -4 0>, rot=<0 0 0 -1>}
 deform_knee_out_r {parent=deform_knee_out_base_r, loc=<0 0 0>, rot=<0 0 0 1>}
 thigh_twist_01_r {parent=thigh_r, loc=<27.952 0 0>, rot=<0.0013 0 0 -1>}
 deform_groin_base_r {parent=thigh_r, loc=<-0.4801 -5.6776 0.2486>, rot=<0 0 0 -1>}
 deform_groin_r {parent=deform_groin_base_r, loc=<0 0 0>, rot=<0 0 0 1>}
 wrench_base {parent=thigh_r, loc=<1.5165 -2.831 6.6437>, rot=<-0.2139 0.118 -0.0733 -0.9669>}
 wrench_dyn {parent=wrench_base, loc=<0 0 0>, rot=<0 0 0 1>}
 Strap_corr_base_r {parent=thigh_r, loc=<0 0 0>, rot=<-0.7064 0.0308 -0.0308 -0.7064>}
 Strap_corr_r {parent=Strap_corr_base_r, loc=<26.2066 7.7005 -0.3034>, rot=<0 0 0 1>}
 jumpsuit_Base_r {parent=thigh_r, loc=<6.6722 -9.1908 -2.1008>, rot=<-0.6759 0.1497 -0.2054 -0.6918>}
 Strap_01_dyn_r {parent=jumpsuit_Base_r, loc=<0 0 0>, rot=<0 0 0 1>}
 Strap_02_dyn_r {parent=Strap_01_dyn_r, loc=<6.4184 0 0>, rot=<0 0 0 1>}
 Strap_03_dyn_r {parent=Strap_02_dyn_r, loc=<5.439 0 0>, rot=<0.0577 -0.1322 -0.0077 -0.9895>}
 Strap_04_dyn_r {parent=Strap_03_dyn_r, loc=<6.411 0 0>, rot=<0.1907 -0.2948 -0.0029 -0.9363>}
 Strap_05_dyn_r {parent=Strap_04_dyn_r, loc=<7.0372 0 0>, rot=<0 0 0 1>}
 deform_glute_base_l {parent=pelvis, loc=<-7.3466 -9.0213 -9.1245>, rot=<0.0172 0.031 -0.4477 -0.8935>}
 deform_glute_l {parent=deform_glute_base_l, loc=<0 0 0>, rot=<0 0 0 1>}
 deform_glute_base_r {parent=pelvis, loc=<-7.3464 -9.0213 9.1245>, rot=<0.031 -0.0172 0.8935 -0.4477>}
 deform_glute_r {parent=deform_glute_base_r, loc=<0 0 0>, rot=<0 0 0 1>}
 Strap_Back_base_r {parent=pelvis, loc=<7.6278 -9.5608 5>, rot=<0.1896 -0.6812 0.7038 -0.0684>}
 Strap_Back_base_l {parent=pelvis, loc=<7.6278 -9.5608 -4>, rot=<0.0684 -0.7038 0.6812 -0.1896>}
 Strap_Back_02_Base_l {parent=pelvis, loc=<0.2779 -12.398 -5.3892>, rot=<-0.1098 -0.7131 0.652 -0.2332>}
 Strap_Back_02_dyn_l {parent=Strap_Back_02_Base_l, loc=<0 0 0>, rot=<-0.0002 -0.0013 -0.0001 -1>}
 Strap_Back_03_dyn_l {parent=Strap_Back_02_dyn_l, loc=<5.2505 0 0>, rot=<0.0049 -0.2143 0.0225 -0.9765>}
 Strap_Back_04_dyn_l {parent=Strap_Back_03_dyn_l, loc=<8.097 0 0>, rot=<-0.0012 -0.0632 -0.019 -0.9978>}
 Strap_Back_05_dyn_l {parent=Strap_Back_04_dyn_l, loc=<5.0111 0 0>, rot=<0.0705 -0.0292 -0.0334 -0.9965>}
 Strap_Back_06_dyn_l {parent=Strap_Back_05_dyn_l, loc=<5.9128 0 0>, rot=<0.2555 -0.1029 -0.0637 -0.9592>}
 Strap_Back_07_dyn_l {parent=Strap_Back_06_dyn_l, loc=<8.5192 0 0>, rot=<0 0 0 1>}
 Strap_Back_02_Base_r {parent=pelvis, loc=<0.2779 -12.398 6.3892>, rot=<0.1701 -0.6673 0.7215 0.0726>}
 Strap_Back_02_dyn_r {parent=Strap_Back_02_Base_r, loc=<0 0 0>, rot=<0.0141 0.0014 -0.0061 -0.9999>}
 Strap_Back_03_dyn_r {parent=Strap_Back_02_dyn_r, loc=<5.2505 0 0>, rot=<0.004 -0.1857 0.0209 -0.9824>}
 Strap_Back_04_dyn_r {parent=Strap_Back_03_dyn_r, loc=<8.097 0 0>, rot=<0 0 0.0175 -0.9998>}
 Strap_Back_05_dyn_r {parent=Strap_Back_04_dyn_r, loc=<7.0969 0 0>, rot=<-0.0937 -0.1616 0.0959 -0.9777>}
 Strap_Back_06_dyn_r {parent=Strap_Back_05_dyn_r, loc=<7.6807 0 0>, rot=<-0.2015 -0.1685 -0.0123 -0.9648>}
 Strap_Back_07_dyn_r {parent=Strap_Back_06_dyn_r, loc=<3.9642 0 0>, rot=<0 0 0 1>}
]
variables [
 Backward Movement {type=float, default=0}
 Crouch {type=double, default=0}
 CrouchWalk {type=double, default=0}
 Forward Backward Movement {type=double, default=0}
 Forward Movement {type=float, default=0}
 IsEmote {type=bool, default=False}
 Left Right Movement {type=double, default=0}
 MoveSpeed {type=double, default=0}
 Thigh_L_Side {type=float, default=0}
 Thigh_L_frv {type=float, default=0}
 Thigh_R_Side {type=float, default=0}
 Thigh_R_frv {type=float, default=0}
 isFrEnd {type=bool, default=False}
 isSkyDive {type=bool, default=False}
]
