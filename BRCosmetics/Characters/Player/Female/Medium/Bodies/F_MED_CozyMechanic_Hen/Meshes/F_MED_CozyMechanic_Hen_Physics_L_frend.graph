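# Physics asset for the left-side dangling straps of the Cozy Mechanic Hen female player outfit body.
# Two capsule chains, a front strap from a jumpsuit base locked to the thigh and a back strap from a base locked to the pelvis, swing and twist within angular limits.
# Both chains end at a shared corrective bone with free rotation, tied to the thigh via its own base.
# Pelvis and left thigh capsules provide the body volumes.
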PhysicsAsset /BRCosmetics/Characters/Player/Female/Medium/Bodies/F_MED_CozyMechanic_Hen/Meshes/F_MED_CozyMechanic_Hen_Physics_L_frend
bodies {
 Strap_01_dyn_l [{shapeType=Capsule, length=3.8074, radius=2.8663, center=<0 0 2>, rotation=<rot 90 0 0>}]
 Strap_02_dyn_l [{shapeType=Capsule, length=6.5248, radius=2.8663, center=<0 0 2>, rotation=<rot 86.144 -0.0009 -0.0009>}]
 Strap_03_dyn_l [{shapeType=Capsule, length=5.3043, radius=3.377, center=<0 0 2>, rotation=<rot 86.599 -179.9993 -179.9993>}]
 Strap_04_dyn_l [{shapeType=Capsule, length=5.4388, radius=3.4751, center=<0 0 2>, rotation=<rot 71.206 -179.9999 -179.9999>}]
 Strap_05_dyn_l [{shapeType=Capsule, length=7.008, radius=2.8663, center=<-0.0658 0.045 2.5>, rotation=<rot 65.3293 -34.4158 -32.5526>}]
 Strap_Back_02_Base_l [{shapeType=Capsule, length=1, radius=1}]
 Strap_Back_02_dyn_l [{shapeType=Capsule, length=5.0205, radius=2.7171, center=<2 0 2>, rotation=<rot 86.3458 -0.0006 -0.0006>}]
 Strap_Back_03_dyn_l [{shapeType=Capsule, length=5.1708, radius=3.3943, center=<2 0 2>, rotation=<rot 90 0 0>}]
 Strap_Back_04_dyn_l [{shapeType=Capsule, length=4.6812, radius=4.0846, center=<2 0 2.5>, rotation=<rot 74.7411 -179.9998 -179.9999>}]
 Strap_Back_05_dyn_l [{shapeType=Capsule, length=4.1107, radius=3.0156, center=<3 0 3>, rotation=<rot 82.8488 -179.9996 -179.9996>}]
 Strap_Back_06_dyn_l [{shapeType=Capsule, length=3.8626, radius=2.9391, center=<2 0 3.4>, rotation=<rot 90 0 0>}]
 Strap_Back_07_dyn_l [{shapeType=Capsule, length=6.6016, radius=3.334, center=<0 0 3>, rotation=<rot 81.1262 -0.0001 -11.4478>}]
 Strap_Back_base_l [{shapeType=Capsule, length=0.505, radius=0.505, shapeName=Strap_Back_base_l_capsule, rotation=<rot 0 -90 0>}]
 Strap_corr_base_l [{shapeType=Capsule, length=3.3667, radius=1}]
 Strap_corr_l [{shapeType=Capsule, length=1, radius=1}]
 jumpsuit_Base_l [{shapeType=Capsule, length=1, radius=1}]
 pelvis [
  {shapeType=Capsule, length=17.407, radius=9.192, center=<0 -0.5244 0>}
  {shapeType=Capsule, length=9.9046, radius=11.1083, shapeName=pelvis_capsule, center=<-3.4923 -0.8855 0>}
 ]
 thigh_l [
  {
   shapeType=Capsule
   length=14.5516
   radius=8.9882
   shapeName=thigh_l_capsule
   center=<-11.2739 -1.8246 1.1659>
   rotation=<rot 1.126 -98.5485 -86.087>
  }
  {
   shapeType=Capsule
   length=12.5055
   radius=5.0098
   shapeName=thigh_l_capsule
   center=<-14.6211 0.0001 -14.2414>
   rotation=<rot -80 -0.0263 90.0161>
  }
 ]
 thigh_twist_01_l [{shapeType=Capsule, length=39.9758, radius=8.7907, center=<1.403 -1.1684 1.0683>, rotation=<rot 90 0.7463 0>}]
}
constraints [
 {bone1Name=Strap_02_dyn_l, bone2Name=Strap_01_dyn_l, swing1LimitDegrees=45, swing2LimitDegrees=45, twistLimitDegrees=10}
 {bone1Name=Strap_03_dyn_l, bone2Name=Strap_02_dyn_l, swing1LimitDegrees=40, swing2LimitDegrees=60, twistLimitDegrees=20}
 {bone1Name=Strap_04_dyn_l, bone2Name=Strap_03_dyn_l, swing1LimitDegrees=50, swing2LimitDegrees=50, twistLimitDegrees=20}
 {bone1Name=Strap_05_dyn_l, bone2Name=Strap_04_dyn_l, swing1LimitDegrees=45, swing2LimitDegrees=45, twistLimitDegrees=6}
 {
  bone1Name=Strap_corr_l
  bone2Name=Strap_corr_base_l
  swing1LimitDegrees=45
  swing1Motion=Locked
  swing2LimitDegrees=45
  swing2Motion=Locked
  twistLimitDegrees=45
  twistMotion=Locked
 }
 {
  bone1Name=thigh_twist_01_l
  bone2Name=thigh_l
  swing1LimitDegrees=45
  swing1Motion=Locked
  swing2LimitDegrees=45
  swing2Motion=Locked
  twistLimitDegrees=45
  twistMotion=Locked
 }
 {bone1Name=Strap_corr_base_l, bone2Name=thigh_l, swing1LimitDegrees=10, swing2LimitDegrees=20, twistLimitDegrees=0}
 {bone1Name=Strap_Back_07_dyn_l, bone2Name=Strap_Back_06_dyn_l, swing1LimitDegrees=45, swing2LimitDegrees=45, twistLimitDegrees=20}
 {
  bone1Name=Strap_Back_base_l
  bone2Name=pelvis
  swing1LimitDegrees=20
  swing1Motion=Locked
  swing2LimitDegrees=20
  swing2Motion=Locked
  twistLimitDegrees=2
  twistMotion=Locked
 }
 {bone1Name=Strap_Back_03_dyn_l, bone2Name=Strap_Back_02_dyn_l, swing1LimitDegrees=10, swing2LimitDegrees=10, twistLimitDegrees=30}
 {bone1Name=Strap_Back_04_dyn_l, bone2Name=Strap_Back_03_dyn_l, swing1LimitDegrees=10, swing2LimitDegrees=10, twistLimitDegrees=20}
 {bone1Name=Strap_Back_05_dyn_l, bone2Name=Strap_Back_04_dyn_l, swing1LimitDegrees=10, swing2LimitDegrees=10, twistLimitDegrees=40}
 {bone1Name=Strap_Back_06_dyn_l, bone2Name=Strap_Back_05_dyn_l, swing1LimitDegrees=10, swing2LimitDegrees=10, twistLimitDegrees=30}
 {bone1Name=Strap_Back_02_dyn_l, bone2Name=Strap_Back_02_Base_l, swing1LimitDegrees=30, swing2LimitDegrees=30, twistLimitDegrees=30}
 {
  bone1Name=Strap_Back_02_Base_l
  bone2Name=pelvis
  swing1LimitDegrees=45
  swing1Motion=Locked
  swing2LimitDegrees=45
  swing2Motion=Locked
  twistLimitDegrees=45
  twistMotion=Locked
 }
 {bone1Name=Strap_01_dyn_l, bone2Name=jumpsuit_Base_l, swing1LimitDegrees=60, swing2LimitDegrees=50, twistLimitDegrees=20}
 {
  bone1Name=jumpsuit_Base_l
  bone2Name=thigh_l
  swing1LimitDegrees=45
  swing1Motion=Locked
  swing2LimitDegrees=45
  swing2Motion=Locked
  twistLimitDegrees=45
  twistMotion=Locked
 }
 {
  bone1Name=Strap_05_dyn_l
  bone2Name=Strap_corr_l
  swing1LimitDegrees=45
  swing1Motion=Free
  swing2LimitDegrees=45
  swing2Motion=Free
  twistLimitDegrees=45
  twistMotion=Free
 }
 {
  bone1Name=Strap_Back_07_dyn_l
  bone2Name=Strap_corr_l
  swing1LimitDegrees=45
  swing1Motion=Free
  swing2LimitDegrees=45
  swing2Motion=Free
  twistLimitDegrees=45
  twistMotion=Free
 }
 {
  bone1Name=thigh_l
  bone2Name=pelvis
  swing1LimitDegrees=45
  swing1Motion=Free
  swing2LimitDegrees=45
  swing2Motion=Free
  twistLimitDegrees=45
  twistMotion=Free
 }
]
