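# Physics asset for the right-side dangling jumpsuit straps on the female CozyMechanic_Hen player cosmetic, built entirely from capsule bodies.
# Front and back strap chains hang from base bodies locked to the pelvis, using angular limits between segments.
# The chain ends join each other, closing a loop, and the front end also links to a corrective strap body with free rotation.
# Pelvis, right thigh, calf and wrench capsules supply surrounding body shapes.
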PhysicsAsset /BRCosmetics/Characters/Player/Female/Medium/Bodies/F_MED_CozyMechanic_Hen/Meshes/F_MED_CozyMechanic_Hen_Physics_R
bodies {
 Strap_01_dyn_r [{shapeType=Capsule, length=3.8074, radius=2.8663, center=<2.8095 0 2>, rotation=<rot 90 0 0>}]
 Strap_02_dyn_r [{shapeType=Capsule, length=5.1096, radius=3.2635, center=<2.9141 0 2.0202>, rotation=<rot 90 0 0>}]
 Strap_03_dyn_r [{shapeType=Capsule, length=5.5222, radius=3.3655, center=<3.0778 0 1.9598>, rotation=<rot 87.781 179.9987 179.9988>}]
 Strap_04_dyn_r [{shapeType=Capsule, length=5.6219, radius=3.0846, center=<2.5138 -0.0002 2.2177>, rotation=<rot 89.8398 179.5279 179.5336>}]
 Strap_05_dyn_r [{shapeType=Capsule, length=5.357, radius=3.6571, center=<0 0 3>, rotation=<rot 74.4635 57.5066 47.201>}]
 Strap_Back_02_Base_r [{shapeType=Capsule, length=1, radius=1}]
 Strap_Back_02_dyn_r [{shapeType=Capsule, length=4.7284, radius=3.0093, center=<2 0 2>, rotation=<rot 86.3458 0.0004 0.0004>}]
 Strap_Back_03_dyn_r [{shapeType=Capsule, length=3.6847, radius=2.8186, center=<2 0 2>, rotation=<rot 90 0 0>}]
 Strap_Back_04_dyn_r [{shapeType=Capsule, length=5.5342, radius=3.0753, center=<2 0 2>, rotation=<rot 90 0 0>}]
 Strap_Back_05_dyn_r [{shapeType=Capsule, length=6.6597, radius=3.4938, center=<3 0 2>, rotation=<rot 90 0 0>}]
 Strap_Back_06_dyn_r [{shapeType=Capsule, length=7.8402, radius=3.6182, center=<1.9772 0 2.4003>, rotation=<rot 77.2153 179.9996 179.9998>}]
 Strap_Back_07_dyn_r [{shapeType=Capsule, length=3.8589, radius=3.4638, center=<1.6916 0 3>, rotation=<rot 87.5061 135.6423 140.0889>}]
 Strap_corr_base_l [{shapeType=Capsule, length=3.3667, radius=1}]
 Strap_corr_base_r [{shapeType=Capsule, length=3.3667, radius=1}]
 Strap_corr_r [{shapeType=Capsule, length=5.128, radius=2.0635, center=<0 2.001 0>}]
 calf_r [{shapeType=Capsule, length=14.9913, radius=5.3282, shapeName=calf_r_capsule, center=<15.0708 4.0639 0.8206>, rotation=<rot 0 97.727 -87.729>}]
 jumpsuit_Base_r [{shapeType=Capsule, length=1, radius=1}]
 pelvis [{shapeType=Capsule, length=12.3274, radius=10.8242, shapeName=pelvis_capsule, center=<-3.4923 -1.4394 0>}]
 thigh_twist_01_r [{shapeType=Capsule, length=34.694, radius=8.9031, center=<-1.3867 -0.1082 -0.9638>, rotation=<rot 88.4387 -1.3391 -179.9998>}]
 wrench_base [{shapeType=Capsule, length=6.6789, radius=5.0623, shapeName=wrench_base_capsule, center=<6.6661 0 -0.0895>, rotation=<rot 0 -90.0001 90.7696>}]
}
constraints [
 {bone1Name=Strap_05_dyn_r, bone2Name=Strap_04_dyn_r, swing1LimitDegrees=60, swing2LimitDegrees=60, twistLimitDegrees=20}
 {bone1Name=Strap_02_dyn_r, bone2Name=Strap_01_dyn_r, swing1LimitDegrees=50, swing2LimitDegrees=50, twistLimitDegrees=20}
 {bone1Name=Strap_03_dyn_r, bone2Name=Strap_02_dyn_r, swing1LimitDegrees=40, swing2LimitDegrees=60, twistLimitDegrees=20}
 {bone1Name=Strap_04_dyn_r, bone2Name=Strap_03_dyn_r, swing1LimitDegrees=50, swing2LimitDegrees=50, twistLimitDegrees=20}
 {
  bone1Name=Strap_corr_r
  bone2Name=Strap_corr_base_r
  swing1LimitDegrees=45
  swing1Motion=Locked
  swing2LimitDegrees=45
  swing2Motion=Locked
  twistLimitDegrees=45
  twistMotion=Locked
 }
 {bone1Name=Strap_corr_base_l, bone2Name=pelvis, swing1LimitDegrees=10, swing2LimitDegrees=20, twistLimitDegrees=0}
 {
  bone1Name=thigh_twist_01_r
  bone2Name=pelvis
  swing1LimitDegrees=45
  swing1Motion=Locked
  swing2LimitDegrees=45
  swing2Motion=Locked
  twistLimitDegrees=45
  twistMotion=Locked
 }
 {bone1Name=Strap_corr_base_r, bone2Name=pelvis, swing1LimitDegrees=10, swing2LimitDegrees=20, twistLimitDegrees=0}
 {bone1Name=Strap_Back_07_dyn_r, bone2Name=Strap_Back_06_dyn_r, swing1LimitDegrees=40, swing2LimitDegrees=40, twistLimitDegrees=20}
 {bone1Name=Strap_Back_03_dyn_r, bone2Name=Strap_Back_02_dyn_r, swing1LimitDegrees=40, swing2LimitDegrees=40, twistLimitDegrees=30}
 {bone1Name=Strap_Back_04_dyn_r, bone2Name=Strap_Back_03_dyn_r, swing1LimitDegrees=50, swing2LimitDegrees=50, twistLimitDegrees=50}
 {bone1Name=Strap_Back_05_dyn_r, bone2Name=Strap_Back_04_dyn_r, swing1LimitDegrees=50, swing2LimitDegrees=80, twistLimitDegrees=50}
 {bone1Name=Strap_Back_06_dyn_r, bone2Name=Strap_Back_05_dyn_r, swing1LimitDegrees=50, swing2LimitDegrees=50, twistLimitDegrees=30}
 {bone1Name=Strap_Back_02_dyn_r, bone2Name=Strap_Back_02_Base_r, swing1LimitDegrees=20, swing2LimitDegrees=40, twistLimitDegrees=30}
 {
  bone1Name=Strap_Back_02_Base_r
  bone2Name=pelvis
  swing1LimitDegrees=45
  swing1Motion=Locked
  swing2LimitDegrees=45
  swing2Motion=Locked
  twistLimitDegrees=45
  twistMotion=Locked
 }
 {bone1Name=Strap_01_dyn_r, bone2Name=jumpsuit_Base_r, swing1LimitDegrees=60, swing2LimitDegrees=50, twistLimitDegrees=20}
 {
  bone1Name=jumpsuit_Base_r
  bone2Name=pelvis
  swing1LimitDegrees=45
  swing1Motion=Locked
  swing2LimitDegrees=45
  swing2Motion=Locked
  twistLimitDegrees=45
  twistMotion=Locked
 }
 {
  bone1Name=wrench_base
  bone2Name=pelvis
  swing1LimitDegrees=45
  swing1Motion=Locked
  swing2LimitDegrees=45
  swing2Motion=Locked
  twistLimitDegrees=45
  twistMotion=Locked
 }
 {
  bone1Name=calf_r
  bone2Name=pelvis
  swing1LimitDegrees=45
  swing1Motion=Free
  swing2LimitDegrees=45
  swing2Motion=Free
  twistLimitDegrees=45
  twistMotion=Free
 }
 {bone1Name=Strap_Back_07_dyn_r, bone2Name=Strap_05_dyn_r, swing1LimitDegrees=60, swing2LimitDegrees=60, twistLimitDegrees=60}
 {
  bone1Name=Strap_05_dyn_r
  bone2Name=Strap_corr_r
  swing1LimitDegrees=45
  swing1Motion=Free
  swing2LimitDegrees=45
  swing2Motion=Free
  twistLimitDegrees=45
  twistMotion=Free
 }
]
