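# Post-process style cosmetic AnimBP for the female CozyMechanic_Hen body that adds secondary motion to the pose copied from its parent mesh.
# A Control Rig fed by movement, emote, skydive and frontend states feeds split left/right RigidBody sims in frontend outside emotes, otherwise one combined sim.
# AnimDynamics then drive wrench, zipper and torso bones, with a weighted zipper-base constraint.
# The event graph's computed GravityOverride is unused by the AnimGraph.
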AnimBlueprint /BRCosmetics/Characters/Player/Female/Medium/Bodies/F_MED_CozyMechanic_Hen/Meshes/F_MED_CozyMechanic_Hen_AnimBP
target_skeleton=&/F_MED_CozyMechanic_Hen_Skeleton
graphs [
 AnimGraph {nodes [
   n1 = RigidBody!(ComponentPose, Alpha, OverrideWorldGravity:Vector="0,0,-980", ExternalForce) -> (Pose) {
    bClampLinearTranslationLimitToRefPose=true
    bOverrideWorldGravity=true
    overridePhysicsAsset=&/F_MED_CozyMechanic_Hen_Physics_L
    simulationSpace=BaseBoneSpace
    alphaBoolBlend {blendOption=Linear}
    baseBoneRef {boneName=pelvis}
    overrideWorldGravity {z=-980}
    simSpaceSettings {
     maxAngularAcceleration=40
     maxAngularVelocity=3
     maxLinearAcceleration=500
     maxLinearVelocity=400
     worldAlpha=1
     externalLinearDragV=<0.2 0.2 0.2>
    }
   }
   RigidBody(ComponentPose=@n1, OverrideWorldGravity="0,0,-980") {
    bClampLinearTranslationLimitToRefPose=true
    bOverrideWorldGravity=true
    overridePhysicsAsset=&/F_MED_CozyMechanic_Hen_Physics_R
    simulationSpace=BaseBoneSpace
    alphaBoolBlend {blendOption=Linear}
    baseBoneRef {boneName=pelvis}
    overrideWorldGravity {z=-980}
    simSpaceSettings {
     maxAngularAcceleration=40
     maxAngularVelocity=3
     maxLinearAcceleration=500
     maxLinearVelocity=400
     worldAlpha=1
     externalLinearDragV=<0.2 0.2 0.2>
    }
   }
   RigidBody!(ComponentPose, Alpha, ExternalForce) -> (Pose) {
    bClampLinearTranslationLimitToRefPose=true
    bOverrideWorldGravity=true
    overridePhysicsAsset=&/F_MED_CozyMechanic_Hen_Physics_Both
    simulationSpace=BaseBoneSpace
    alphaBoolBlend {blendOption=Linear}
    baseBoneRef {boneName=pelvis}
    overrideWorldGravity {z=-980}
    simSpaceSettings {
     maxAngularAcceleration=40
     maxAngularVelocity=3
     maxLinearAcceleration=500
     maxLinearVelocity=400
     worldAlpha=1
     externalLinearDragV=<0.2 0.2 0.2>
    }
   }
   n2 = RigidBody!(ComponentPose, Alpha, OverrideWorldGravity:Vector="0,0,-980", ExternalForce) -> (Pose) {
    bClampLinearTranslationLimitToRefPose=true
    bOverrideWorldGravity=true
    overridePhysicsAsset=&F_MED_CozyMechanic_Hen_Physics_L
    simulationSpace=BaseBoneSpace
    alphaBoolBlend {blendOption=Linear}
    baseBoneRef {boneName=pelvis}
    overrideWorldGravity {z=-980}
    simSpaceSettings {
     maxAngularAcceleration=40
     maxAngularVelocity=3
     maxLinearAcceleration=500
     maxLinearVelocity=400
     worldAlpha=1
     externalLinearDragV=<0.2 0.2 0.2>
    }
   }
   RigidBody(ComponentPose=@n2, OverrideWorldGravity="0,0,-980") {
    bClampLinearTranslationLimitToRefPose=true
    bOverrideWorldGravity=true
    overridePhysicsAsset=&F_MED_CozyMechanic_Hen_Physics_R
    simulationSpace=BaseBoneSpace
    alphaBoolBlend {blendOption=Linear}
    baseBoneRef {boneName=pelvis}
    overrideWorldGravity {z=-980}
    simSpaceSettings {
     maxAngularAcceleration=40
     maxAngularVelocity=3
     maxLinearAcceleration=500
     maxLinearVelocity=400
     worldAlpha=1
     externalLinearDragV=<0.2 0.2 0.2>
    }
   }
   n3 = RigidBody(ComponentPose=Local To Component(LocalPose=Use cached pose 'SavedPoseFrEn'()), OverrideWorldGravity="0,0,-980") {
    bClampLinearTranslationLimitToRefPose=true
    bOverrideWorldGravity=true
    overridePhysicsAsset=&F_MED_CozyMechanic_Hen_Physics_Both
    simulationSpace=BaseBoneSpace
    alphaBoolBlend {blendOption=Linear}
    baseBoneRef {boneName=pelvis}
    overrideWorldGravity {z=-980}
    simSpaceSettings {
     maxAngularAcceleration=40
     maxAngularVelocity=3
     maxLinearAcceleration=500
     maxLinearVelocity=400
     worldAlpha=0.8
     externalLinearDragV=<0.2 0.2 0.2>
    }
   }
   n4 = Component To Local(ComponentPose=@n3)
   n5 = RigidBody!(ComponentPose=Local To Component(LocalPose=Use cached pose 'SavedPoseFrEn'()), Alpha, ExternalForce) -> (Pose) {
    bClampLinearTranslationLimitToRefPose=true
    bOverrideWorldGravity=true
    overridePhysicsAsset=&F_MED_CozyMechanic_Hen_Physics_R
    simulationSpace=BaseBoneSpace
    alphaBoolBlend {blendOption=Linear}
    baseBoneRef {boneName=pelvis}
    overrideWorldGravity {z=-980}
    simSpaceSettings {
     maxAngularAcceleration=40
     maxAngularVelocity=3
     maxLinearAcceleration=500
     maxLinearVelocity=400
     worldAlpha=1
     externalLinearDragV=<0.2 0.2 0.2>
    }
   }
   n6 = RigidBody(ComponentPose=@n5) {
    bClampLinearTranslationLimitToRefPose=true
    bOverrideWorldGravity=true
    overridePhysicsAsset=&/F_MED_CozyMechanic_Hen_Physics_L_frend
    simulationSpace=BaseBoneSpace
    alphaBoolBlend {blendOption=Linear}
    baseBoneRef {boneName=pelvis}
    overrideWorldGravity {z=-980}
    simSpaceSettings {
     maxAngularAcceleration=40
     maxAngularVelocity=3
     maxLinearAcceleration=500
     maxLinearVelocity=400
     worldAlpha=1
     externalLinearDragV=<0.2 0.2 0.2>
    }
   }
   n7 = Component To Local(ComponentPose=@n6)
   n8 = Get Frontend() -> (Frontend:bool)
   n9 = Anim Dynamics - Bone: wrench_dyn(
    ComponentPose=Local To Component(LocalPose=Blend Poses by bool(BlendPose_0=@n7, BlendPose_1=@n4, bActiveValue=@n8) {blendType=Linear})
    Alpha=1
    ExternalForce="0,0,0"
   ) {
    angularDampingOverride=1
    angularSpringConstant=400
    bAngularSpring=true
    bOverrideAngularDamping=true
    bUseGravityOverride=true
    simulationSpace=BoneRelative
    alphaBoolBlend {blendOption=Linear}
    boundBone {boneName=wrench_dyn}
    componentAppliedLinearAccClamp=<1000 1000 1000>
    componentLinearAccScale=<0.7 0.7 0.7>
    componentLinearVelScale=<0.7 0.7 0.7>
    gravityOverride {z=-980}
    physicsBodyDefinitions [{
      boundBone {boneName=wrench_dyn}
      boxExtents=<10 10 10>
      constraintSetup {coneAngle=1, angularLimitsMax=<1 3 5>, angularLimitsMin=<-1 -10 -5>, angularTarget=<1 0 0>}
      localJointOffset=<1 0 0>
     }]
    relativeSpaceBone {boneName=root}
   }
   n10 = Anim Dynamics - Bone: zipper_dyn(ComponentPose=@n9, Alpha=1, ExternalForce="0,0,0") {
    angularDampingOverride=1
    angularSpringConstant=400
    bAngularSpring=true
    bOverrideAngularDamping=true
    bUseGravityOverride=true
    simulationSpace=BoneRelative
    alphaBoolBlend {blendOption=Linear}
    boundBone {boneName=zipper_dyn}
    componentAppliedLinearAccClamp=<1000 1000 1000>
    componentLinearAccScale=<0.7 0.7 0.7>
    componentLinearVelScale=<0.7 0.7 0.7>
    gravityOverride {z=-980}
    physicsBodyDefinitions [{
      boundBone {boneName=zipper_dyn}
      boxExtents=<10 10 10>
      constraintSetup {coneAngle=1, angularLimitsMax=<1 10 20>, angularLimitsMin=<-1 -10 -1>, angularTarget=<1 0 0>}
      localJointOffset=<1 0 0>
     }]
    relativeSpaceBone {boneName=root}
   }
   n11 = "Constraint - zipper_base (PPPPPP)"!(
    ComponentPose=@n10
    Alpha=1
    ConstraintWeights_0=0.11
    ConstraintWeights_1:float=0.2888
    ConstraintWeights_2:float=0.3216
    ConstraintWeights_3:float=0.0571
    ConstraintWeights_4:float=0.0044
    ConstraintWeights_5:float=0.2182
   ) -> (Pose) {
    boneToModify {boneName=zipper_base}
    constraintSetup [
     {targetBone {boneName=head}}
     {targetBone {boneName=neck_02}}
     {targetBone {boneName=neck_01}}
     {targetBone {boneName=spine_05}}
     {targetBone {boneName=spine_04}}
     {targetBone {boneName=C_jaw}}
    ]
    constraintWeights [1,1,1,1,1,1]
   }
   n12 = "Transform (Modify) Bone - Bone: dyn_torso_l"(ComponentPose=@n11, Alpha=1, Translation="0,0,1", Rotation="0,0,0", Scale="1,1,1") {translationMode=BMM_Additive, translationSpace=BCS_BoneSpace, boneToModify {boneName=dyn_torso_l}, translation {z=1}}
   n13 = "Transform (Modify) Bone - Bone: dyn_torso_r"(ComponentPose=@n12, Alpha=1, Translation="0,0,1", Rotation="0,0,0", Scale="1,1,1") {translationMode=BMM_Additive, translationSpace=BCS_BoneSpace, boneToModify {boneName=dyn_torso_r}, translation {z=1}}
   n14 = Anim Dynamics - Bone: dyn_torso_r!(ComponentPose=@n13, Alpha=1, GravityOverride:Vector="0,0,-980", ExternalForce="0,0,0") -> (Pose) {
    angularBiasOverride=0.999
    angularDampingOverride=0.999
    angularSpringConstant=4000
    bAngularSpring=true
    bGravityOverrideInSimSpace=true
    bOverrideAngularBias=true
    bOverrideAngularDamping=true
    bUseGravityOverride=true
    linearDampingOverride=0.9999
    linearSpringConstant=100
    simulationSpace=RootRelative
    boundBone {boneName=dyn_torso_r}
    gravityOverride {z=-980}
    physicsBodyDefinitions [{
      boundBone {boneName=dyn_torso_r}
      boxExtents=<21 21 21>
      constraintSetup {angularLimitsMax=<0 10 15>, angularLimitsMin=<0 -20 -15>, angularTarget=<1 0 0>}
      localJointOffset=<2 0 0>
     }]
    relativeSpaceBone {boneName=spine_04}
    simSpaceSettings {maxAngularAcceleration=55, maxAngularVelocity=8, simSpaceAngularAlpha=0.3}
   }
   n15 = Anim Dynamics - Bone: dyn_torso_l!(ComponentPose=@n14, Alpha=1, GravityOverride:Vector="0,0,-980", ExternalForce="0,0,0") -> (Pose) {
    angularBiasOverride=0.999
    angularDampingOverride=0.999
    angularSpringConstant=4000
    bAngularSpring=true
    bGravityOverrideInSimSpace=true
    bOverrideAngularBias=true
    bOverrideAngularDamping=true
    bUseGravityOverride=true
    linearDampingOverride=0.9999
    linearSpringConstant=100
    simulationSpace=RootRelative
    boundBone {boneName=dyn_torso_l}
    gravityOverride {z=-980}
    physicsBodyDefinitions [{
      boundBone {boneName=dyn_torso_l}
      boxExtents=<20 20 20>
      constraintSetup {angularLimitsMax=<0 10 15>, angularLimitsMin=<0 -20 -15>, angularTarget=<1 0 0>}
      localJointOffset=<2 0 0>
     }]
    relativeSpaceBone {boneName=spine_04}
    simSpaceSettings {maxAngularAcceleration=55, maxAngularVelocity=8, simSpaceAngularAlpha=0.3}
   }
   n16 = Anim Dynamics - Bone: dyn_torso_r(ComponentPose=@n15, Alpha=1, GravityOverride="0,0,-200", ExternalForce="0,0,0") {
    angularBiasOverride=1
    angularDampingOverride=0.99
    angularSpringConstant=10000
    bGravityOverrideInSimSpace=true
    bLinearSpring=true
    bOverrideLinearDamping=true
    bUseGravityOverride=true
    linearDampingOverride=0.9999
    linearSpringConstant=180
    simulationSpace=RootRelative
    boundBone {boneName=dyn_torso_r}
    gravityOverride {z=-200}
    physicsBodyDefinitions [{
      boundBone {boneName=dyn_torso_r}
      boxExtents=<70 70 70>
      constraintSetup {angularLimitsMax=<0 0 0>, angularLimitsMin=<0 0 0>, angularTarget=<1 0 0>, linearAxesMax=<0 0.75 1.5>, linearAxesMin=<0 -0.75 -0.5>}
      localJointOffset=<0.5 0 0>
     }]
    relativeSpaceBone {boneName=spine_04}
    simSpaceSettings {maxAngularAcceleration=400, maxAngularVelocity=200, simSpaceAngularAlpha=0.1}
   }
   n17 = Anim Dynamics - Bone: dyn_torso_l(ComponentPose=@n16, Alpha=1, GravityOverride="0,0,-200", ExternalForce="0,0,0") {
    angularBiasOverride=1
    angularDampingOverride=0.99
    angularSpringConstant=10000
    bGravityOverrideInSimSpace=true
    bLinearSpring=true
    bOverrideLinearDamping=true
    bUseGravityOverride=true
    linearDampingOverride=0.9999
    linearSpringConstant=180
    simulationSpace=RootRelative
    boundBone {boneName=dyn_torso_l}
    gravityOverride {z=-200}
    physicsBodyDefinitions [{
      boundBone {boneName=dyn_torso_l}
      boxExtents=<80 80 80>
      constraintSetup {angularLimitsMax=<0 0 0>, angularLimitsMin=<0 0 0>, angularTarget=<1 0 0>, linearAxesMax=<0 0.75 1.5>, linearAxesMin=<0 -0.75 -0.5>}
      localJointOffset=<0.5 0 0>
     }]
    relativeSpaceBone {boneName=spine_04}
    simSpaceSettings {maxAngularAcceleration=400, maxAngularVelocity=200, simSpaceAngularAlpha=0.1}
   }
   Output Pose(Result=Component To Local(ComponentPose=@n17)) {layerGroup=None}
   n18 = Control Rig!(
    Source=Copy Pose From Mesh() {bUseAttachedParent=true}
    Alpha
    MoveSpeed:double=Get PawnSpeed2D()
    Left Right Movement:double=Get Left Right Movement() -> (Left Right Movement:double)
    Forward Backward Movement:double=Get Forward Backward Movement() -> (Forward Backward Movement:double)
    IsEmote:bool=Get bIsPlayingEmote()
    isSkyDive:bool=Get bIsSkyDiving()
    isFrEnd:bool=Get Frontend()
   ) -> (Pose) {controlRigAssetReference {blueprintRigClass=&/F_MED_CozyMechanic_Hen_CtrlRig_B.F_MED_CozyMechanic_Hen_CtrlRig_B_C}}
   Save cached pose 'SavedPoseFrEn'(Pose=@n18) {cacheName=SavedPoseFrEn}
  ]}
 EventGraph {nodes [
   Event BlueprintUpdateAnimation()
   > n1 = Sequence()
   > n2 = PlayerMovement(AnimInstance=Self-Reference(), XDotProductCoefficient=1, YDotProductCoefficient=1, ZDotProductCoefficient=1)
   > Set Left Right Movement(Left Right Movement:double=@n2) -> (Output_Get:double)
   > Set Forward Backward Movement(Forward Backward Movement:double=@n2.YMovement) -> (Output_Get:double)
   n3 = Get bIsPlayingEmote()
   > Set Frontend(Frontend:bool=AND Boolean(A=Get bIsInFrontEnd(), B=NOT Boolean(A=@n3))) -> (Output_Get:bool)
   n4 = Gravity Override State Function DEPRECATED(
    execute=@n1.then_1
    CharacterPartAnimInstance=Self-Reference()
    GameTimeInSeconds=GetGameTimeInSeconds()
    GravityOverrideParams=Get Gravity Override Params()
   )
   > Set GravityOverride(GravityOverride=@n4)
  ]}
]
variables [
 Forward Backward Movement {type=number}
 Frontend {type=boolean}
 Gravity Override Params {type=GravityOverrideParamsStruct, default {skydive_FloatingDown {gravityOverride=<0 -300 30>}}}
 GravityOverride {type=Vector}
 Left Right Movement {type=number}
 NewVar {type=Vector}
 RBAN Settings Out_Native {type=FortRigidBodyAnimNodeParameters}
 Sim Space Settings {
  type=SimSpaceSettings
  default {
   dampingAlpha=1
   maxAngularAcceleration=10000
   maxAngularVelocity=10000
   maxLinearAcceleration=10000
   maxLinearVelocity=10000
   velocityScaleZ=1
   worldAlpha=0
   externalLinearDragV=<0 0 0>
  }
 }
]
