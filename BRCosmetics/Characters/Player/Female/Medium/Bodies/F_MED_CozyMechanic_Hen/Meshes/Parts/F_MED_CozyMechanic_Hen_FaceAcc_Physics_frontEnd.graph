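# Front-end physics asset for the female CozyMechanic Hen face accessory, simulating two hanging braids.
# Each braid is a five-capsule chain with tight swing and twist limits, attached through a small base body whose angular motion to the head is free.
# Capsules on the head, neck, upper spine, clavicles and upper arms form a body proxy, held rigid to each other by angularly locked constraints.
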PhysicsAsset /BRCosmetics/Characters/Player/Female/Medium/Bodies/F_MED_CozyMechanic_Hen/Meshes/Parts/F_MED_CozyMechanic_Hen_FaceAcc_Physics_frontEnd
bodies {
 base_braid_l [{shapeType=Capsule, length=1, radius=1}]
 base_braid_r [{shapeType=Capsule, length=1, radius=1}]
 clavicle_l [{
   shapeType=Capsule
   length=5.9024
   radius=3.6296
   shapeName=clavicle_l_capsule
   center=<3.7954 -0.4046 -0.1044>
   rotation=<rot -79.8826 177.8758 177.9311>
  }]
 clavicle_r [{shapeType=Capsule, length=0.505, radius=0.505, shapeName=clavicle_r_capsule, rotation=<rot 0 -90 0>}]
 dyn_braid_l_01 [{shapeType=Capsule, length=5.5776, radius=1.8696, center=<3.6864 0 0>, rotation=<rot 89.8999 -179.9895 -179.9895>}]
 dyn_braid_l_02 [{shapeType=Capsule, length=4.5582, radius=2.3291, shapeName=pigtail_2_capsule, center=<4.4099 0 0.1982>, rotation=<rot 0 -92.288 -89.7408>}]
 dyn_braid_l_03 [{
   shapeType=Capsule
   length=2.5513
   radius=2.0653
   shapeName=dyn_braid_l_03_capsule
   center=<3.3266 0 0.1026>
   rotation=<rot 0 91.5241 -85.712>
  }]
 dyn_braid_l_04 [{
   shapeType=Capsule
   length=4.4911
   radius=2.0849
   shapeName=dyn_braid_l_04_capsule
   center=<3.9019 0 0.1309>
   rotation=<rot 0 90.7052 -87.5118>
  }]
 dyn_braid_l_05 [{
   shapeType=Capsule
   length=5.46
   radius=2.7421
   shapeName=dyn_braid_l_05_capsule
   center=<5.699 0 -0.4849>
   rotation=<rot 0 -86.0528 -84.7621>
  }]
 dyn_braid_r_01 [{shapeType=Capsule, length=5.5776, radius=1.8696, center=<3.6864 0 0>, rotation=<rot 89.8999 179.9887 179.9888>}]
 dyn_braid_r_02 [{shapeType=Capsule, length=4.5697, radius=2.3177, shapeName=pigtail_2_capsule, center=<4.4099 0 0.1982>, rotation=<rot 0 -92.288 -89.7408>}]
 dyn_braid_r_03 [{shapeType=Capsule, length=3.8241, radius=1.938, shapeName=dyn_braid_r_03_capsule, center=<4.1468 0 0.0061>, rotation=<rot 0 90.407 -87.5995>}]
 dyn_braid_r_04 [{
   shapeType=Capsule
   length=3.7629
   radius=1.9077
   shapeName=dyn_braid_r_04_capsule
   center=<3.8739 0 0.0745>
   rotation=<rot 0 88.8199 -87.6296>
  }]
 dyn_braid_r_05 [{
   shapeType=Capsule
   length=4.9961
   radius=2.7717
   shapeName=dyn_braid_r_05_capsule
   center=<5.3388 0 -0.3987>
   rotation=<rot 0 -88.6032 -83.6848>
  }]
 head [
  {shapeType=Capsule, length=2.6628, radius=6.8126, center=<2.2751 0 0>, rotation=<rot 90 0 0>}
  {shapeType=Capsule, length=17.3793, radius=11.7749, shapeName=based_hair_capsule, center=<0 -19.4573 0>}
  {shapeType=Capsule, length=15, radius=5.5317, shapeName=head_capsule, center=<6.5783 0 0>}
 ]
 neck_01 [{shapeType=Capsule, length=38.4267, radius=4.6585, shapeName=neck_01_capsule, center=<0.8644 -8.0323 0>, rotation=<rot 0 -20.0001 0>}]
 neck_02 [{shapeType=Capsule, length=1.1869, radius=4.3703, shapeName=neck_02_capsule, rotation=<rot 90 0 0>}]
 spine_02 [{shapeType=Capsule, length=7.5351, radius=9.3848, shapeName=spine_02_capsule, center=<0.267 0 0>}]
 spine_03 [{shapeType=Capsule, length=10.1969, radius=8.5991, shapeName=spine_03_capsule, center=<0.6106 0.2666 0>}]
 spine_04 [{shapeType=Capsule, length=4.3252, radius=11.787, shapeName=spine_04_capsule, center=<-1.3378 0.3557 0>}]
 spine_05 [
  {shapeType=Capsule, length=8.5745, radius=6.9919, shapeName=spine_05_capsule, center=<0 0.5022 0>}
  {shapeType=Capsule, length=17.2383, radius=6.2531, shapeName=spine_05_capsule, center=<5.2433 -3.5121 0>}
 ]
 upperarm_l [{shapeType=Capsule, length=19.2989, radius=4.5452, shapeName=upperarm_l_capsule, center=<9.8981 0 0>, rotation=<rot -90 180 180>}]
 upperarm_r [{shapeType=Capsule, length=18.4389, radius=5.4052, shapeName=upperarm_l_capsule, center=<-9.8981 0.7526 0>, rotation=<rot -90 180 0>}]
}
constraints [
 {bone1Name=dyn_braid_l_01, bone2Name=base_braid_l, swing1LimitDegrees=10, swing2LimitDegrees=10, twistLimitDegrees=4}
 {bone1Name=dyn_braid_l_02, bone2Name=dyn_braid_l_01, swing1LimitDegrees=10, swing2LimitDegrees=10, twistLimitDegrees=5}
 {bone1Name=dyn_braid_l_03, bone2Name=dyn_braid_l_02, swing1LimitDegrees=10, swing2LimitDegrees=10, twistLimitDegrees=5}
 {bone1Name=dyn_braid_l_04, bone2Name=dyn_braid_l_03, swing1LimitDegrees=10, swing2LimitDegrees=10, twistLimitDegrees=5}
 {bone1Name=dyn_braid_l_05, bone2Name=dyn_braid_l_04, swing1LimitDegrees=10, swing2LimitDegrees=10, twistLimitDegrees=5}
 {bone1Name=dyn_braid_r_01, bone2Name=base_braid_r, swing1LimitDegrees=10, swing2LimitDegrees=10, twistLimitDegrees=4}
 {bone1Name=dyn_braid_r_02, bone2Name=dyn_braid_r_01, swing1LimitDegrees=10, swing2LimitDegrees=10, twistLimitDegrees=5}
 {bone1Name=dyn_braid_r_03, bone2Name=dyn_braid_r_02, swing1LimitDegrees=10, swing2LimitDegrees=10, twistLimitDegrees=5}
 {bone1Name=dyn_braid_r_04, bone2Name=dyn_braid_r_03, swing1LimitDegrees=10, swing2LimitDegrees=10, twistLimitDegrees=5}
 {bone1Name=dyn_braid_r_05, bone2Name=dyn_braid_r_04, swing1LimitDegrees=10, swing2LimitDegrees=10, twistLimitDegrees=5}
 {
  bone1Name=base_braid_l
  bone2Name=head
  swing1LimitDegrees=45
  swing1Motion=Free
  swing2LimitDegrees=45
  swing2Motion=Free
  twistLimitDegrees=45
  twistMotion=Free
 }
 {
  bone1Name=base_braid_r
  bone2Name=head
  swing1LimitDegrees=45
  swing1Motion=Free
  swing2LimitDegrees=45
  swing2Motion=Free
  twistLimitDegrees=45
  twistMotion=Free
 }
 {
  bone1Name=head
  bone2Name=neck_02
  swing1LimitDegrees=45
  swing1Motion=Locked
  swing2LimitDegrees=45
  swing2Motion=Locked
  twistLimitDegrees=45
  twistMotion=Locked
 }
 {
  bone1Name=neck_02
  bone2Name=neck_01
  swing1LimitDegrees=45
  swing1Motion=Locked
  swing2LimitDegrees=45
  swing2Motion=Locked
  twistLimitDegrees=45
  twistMotion=Locked
 }
 {
  bone1Name=neck_01
  bone2Name=spine_05
  swing1LimitDegrees=45
  swing1Motion=Locked
  swing2LimitDegrees=45
  swing2Motion=Locked
  twistLimitDegrees=45
  twistMotion=Locked
 }
 {
  bone1Name=spine_05
  bone2Name=spine_04
  swing1LimitDegrees=45
  swing1Motion=Locked
  swing2LimitDegrees=45
  swing2Motion=Locked
  twistLimitDegrees=45
  twistMotion=Locked
 }
 {
  bone1Name=spine_04
  bone2Name=spine_03
  swing1LimitDegrees=45
  swing1Motion=Locked
  swing2LimitDegrees=45
  swing2Motion=Locked
  twistLimitDegrees=45
  twistMotion=Locked
 }
 {
  bone1Name=clavicle_l
  bone2Name=spine_05
  swing1LimitDegrees=45
  swing1Motion=Locked
  swing2LimitDegrees=45
  swing2Motion=Locked
  twistLimitDegrees=45
  twistMotion=Locked
 }
 {
  bone1Name=upperarm_l
  bone2Name=clavicle_l
  swing1LimitDegrees=45
  swing1Motion=Locked
  swing2LimitDegrees=45
  swing2Motion=Locked
  twistLimitDegrees=45
  twistMotion=Locked
 }
 {
  bone1Name=clavicle_r
  bone2Name=spine_05
  swing1LimitDegrees=45
  swing1Motion=Locked
  swing2LimitDegrees=45
  swing2Motion=Locked
  twistLimitDegrees=45
  twistMotion=Locked
 }
 {
  bone1Name=upperarm_r
  bone2Name=clavicle_r
  swing1LimitDegrees=45
  swing1Motion=Locked
  swing2LimitDegrees=45
  swing2Motion=Locked
  twistLimitDegrees=45
  twistMotion=Locked
 }
 {
  bone1Name=spine_03
  bone2Name=spine_02
  swing1LimitDegrees=45
  swing1Motion=Locked
  swing2LimitDegrees=45
  swing2Motion=Locked
  twistLimitDegrees=45
  twistMotion=Locked
 }
 {
  bone1Name=upperarm_r
  bone2Name=spine_05
  swing1LimitDegrees=45
  swing1Motion=Locked
  swing2LimitDegrees=45
  swing2Motion=Locked
  twistLimitDegrees=45
  twistMotion=Locked
 }
]
