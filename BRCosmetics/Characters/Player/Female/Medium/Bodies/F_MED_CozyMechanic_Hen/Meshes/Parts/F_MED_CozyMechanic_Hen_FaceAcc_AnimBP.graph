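# Post-process style AnimBP that adds secondary motion to the Cozy Mechanic Hen outfit's face accessory and hair.
# A cached copy of the parent mesh pose, passed through a skydive-aware Control Rig, feeds three RigidBody sims selected for front end, skydiving or default.
# Per-bone spring AnimDynamics then drive left and right hair, followed by fixed additive corrective-bone offsets.
# The update event sets front-end flags and computes gravity override using IntenseCello's parameter asset.
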AnimBlueprint /BRCosmetics/Characters/Player/Female/Medium/Bodies/F_MED_CozyMechanic_Hen/Meshes/Parts/F_MED_CozyMechanic_Hen_FaceAcc_AnimBP
target_skeleton=&/F_MED_CozyMechanic_Hen_FaceAcc_Skeleton
graphs [
 AnimGraph {nodes [
   n1 = RigidBody!(ComponentPose=Local To Component(LocalPose=Use cached pose 'cachPOse'()), Alpha, OverrideWorldGravity:Vector="0,0,0", ExternalForce) -> (Pose) {
    bClampLinearTranslationLimitToRefPose=true
    bForceDisableCollisionBetweenConstraintBodies=true
    bOverrideWorldGravity=true
    overridePhysicsAsset=&/F_MED_CozyMechanic_Hen_FaceAcc_Physics
    simulationSpace=BaseBoneSpace
    alphaBoolBlend {blendOption=Linear}
    baseBoneRef {boneName=spine_05}
    simSpaceSettings {
     dampingAlpha=0
     maxAngularAcceleration=80
     maxAngularVelocity=8
     maxLinearAcceleration=700
     maxLinearVelocity=30
     velocityScaleZ=0.7
     worldAlpha=1
     externalLinearDragV {x=0.5}
    }
   }
   n2 = RigidBody(ComponentPose=Local To Component(LocalPose=Use cached pose 'cachPOse'()), OverrideWorldGravity="0,0,0") {
    bClampLinearTranslationLimitToRefPose=true
    bForceDisableCollisionBetweenConstraintBodies=true
    bOverrideWorldGravity=true
    overridePhysicsAsset=&/F_MED_CozyMechanic_Hen_FaceAcc_SkyDive
    simulationSpace=BaseBoneSpace
    alphaBoolBlend {blendOption=Linear}
    baseBoneRef {boneName=spine_05}
    simSpaceSettings {
     dampingAlpha=0
     maxAngularAcceleration=80
     maxAngularVelocity=8
     maxLinearAcceleration=700
     maxLinearVelocity=30
     velocityScaleZ=0.7
     worldAlpha=1
     externalLinearDragV {x=0.5}
    }
   }
   n3 = Blend Poses by bool(BlendPose_0=Component To Local(ComponentPose=@n2), BlendPose_1=Component To Local(ComponentPose=@n1), bActiveValue=Get bIsSkyDiving()) {blendType=Linear}
   n4 = RigidBody(ComponentPose=Local To Component(LocalPose=Use cached pose 'cachPOse'()), OverrideWorldGravity="0,0,0") {
    bClampLinearTranslationLimitToRefPose=true
    bForceDisableCollisionBetweenConstraintBodies=true
    bOverrideWorldGravity=true
    overridePhysicsAsset=&/F_MED_CozyMechanic_Hen_FaceAcc_Physics_frontEnd
    simulationSpace=BaseBoneSpace
    alphaBoolBlend {blendOption=Linear}
    baseBoneRef {boneName=spine_05}
    simSpaceSettings {
     dampingAlpha=0
     maxAngularAcceleration=80
     maxAngularVelocity=8
     maxLinearAcceleration=700
     maxLinearVelocity=30
     velocityScaleZ=0.7
     worldAlpha=1
     externalLinearDragV {x=0.5}
    }
   }
   n5 = Component To Local(ComponentPose=@n4)
   n6 = Get Frontend() -> (Frontend:bool)
   n7 = Anim Dynamics - Bone: Hair_01_dyn_l(
    ComponentPose=Local To Component(LocalPose=Blend Poses by bool(BlendPose_0=@n5, BlendPose_1=@n3, bActiveValue=@n6) {blendType=Linear})
    Alpha=1
    ExternalForce="0,0,0"
   ) {
    angularDampingOverride=1
    angularSpringConstant=500
    bAngularSpring=true
    bOverrideAngularDamping=true
    bUseGravityOverride=true
    simulationSpace=RootRelative
    alphaBoolBlend {blendOption=Linear}
    boundBone {boneName=Hair_01_dyn_l}
    componentAppliedLinearAccClamp=<5000 5000 5000>
    componentLinearAccScale=<1 1 1>
    componentLinearVelScale=<1 1 1>
    gravityOverride {z=-980}
    physicsBodyDefinitions [{
      boundBone {boneName=Hair_01_dyn_l}
      boxExtents=<12 12 12>
      constraintSetup {angularLimitsMax=<0 1 12>, angularLimitsMin=<0 -15 0>, angularTarget=<1 0 0>}
      localJointOffset=<1 0 0>
     }]
   }
   n8 = Anim Dynamics - Bone: Hair_02_dyn_l(ComponentPose=@n7, Alpha=1, ExternalForce="0,0,0") {
    angularDampingOverride=1
    angularSpringConstant=500
    bAngularSpring=true
    bOverrideAngularDamping=true
    bUseGravityOverride=true
    simulationSpace=RootRelative
    alphaBoolBlend {blendOption=Linear}
    boundBone {boneName=Hair_02_dyn_l}
    componentAppliedLinearAccClamp=<5000 5000 5000>
    componentLinearAccScale=<1 1 1>
    componentLinearVelScale=<1 1 1>
    gravityOverride {z=-980}
    physicsBodyDefinitions [{
      boundBone {boneName=Hair_02_dyn_l}
      boxExtents=<12 12 12>
      constraintSetup {angularLimitsMax=<0 1 13>, angularLimitsMin=<0 0 -13>, angularTarget=<1 0 0>}
      localJointOffset=<1 0 0>
     }]
   }
   n9 = Anim Dynamics - Bone: Hair_03_dyn_l(ComponentPose=@n8, Alpha=1, ExternalForce="0,0,0") {
    angularDampingOverride=1
    angularSpringConstant=500
    bAngularSpring=true
    bOverrideAngularDamping=true
    bUseGravityOverride=true
    simulationSpace=RootRelative
    alphaBoolBlend {blendOption=Linear}
    boundBone {boneName=Hair_03_dyn_l}
    componentAppliedLinearAccClamp=<5000 5000 5000>
    componentLinearAccScale=<1 1 1>
    componentLinearVelScale=<1 1 1>
    gravityOverride {z=-980}
    physicsBodyDefinitions [{
      boundBone {boneName=Hair_03_dyn_l}
      boxExtents=<12 12 12>
      constraintSetup {angularLimitsMax=<0 1 13>, angularLimitsMin=<0 -10 -13>, angularTarget=<1 0 0>}
      localJointOffset=<1 0 0>
     }]
   }
   n10 = Anim Dynamics - Bone: Hair_01_dyn_r(ComponentPose=@n9, Alpha=1, ExternalForce="0,0,0") {
    angularDampingOverride=1
    angularSpringConstant=500
    bAngularSpring=true
    bOverrideAngularDamping=true
    bUseGravityOverride=true
    simulationSpace=RootRelative
    alphaBoolBlend {blendOption=Linear}
    boundBone {boneName=Hair_01_dyn_r}
    componentAppliedLinearAccClamp=<5000 5000 5000>
    componentLinearAccScale=<1 1 1>
    componentLinearVelScale=<1 1 1>
    gravityOverride {z=-980}
    physicsBodyDefinitions [{
      boundBone {boneName=Hair_01_dyn_r}
      boxExtents=<12 12 12>
      constraintSetup {angularLimitsMax=<0 2 8>, angularLimitsMin=<0 -16 -13>, angularTarget=<1 0 0>}
      localJointOffset=<1 0 0>
     }]
   }
   n11 = Anim Dynamics - Bone: Hair_02_dyn_r(ComponentPose=@n10, Alpha=1, ExternalForce="0,0,0") {
    angularDampingOverride=1
    angularSpringConstant=500
    bAngularSpring=true
    bOverrideAngularDamping=true
    bUseGravityOverride=true
    simulationSpace=RootRelative
    alphaBoolBlend {blendOption=Linear}
    boundBone {boneName=Hair_02_dyn_r}
    componentAppliedLinearAccClamp=<5000 5000 5000>
    componentLinearAccScale=<1 1 1>
    componentLinearVelScale=<1 1 1>
    gravityOverride {z=-980}
    physicsBodyDefinitions [{
      boundBone {boneName=Hair_02_dyn_r}
      boxExtents=<12 12 12>
      constraintSetup {angularLimitsMax=<0 2 8>, angularLimitsMin=<0 -14 -13>, angularTarget=<1 0 0>}
      localJointOffset=<1 0 0>
     }]
   }
   n12 = Anim Dynamics - Bone: Hair_03_dyn_r(ComponentPose=@n11, Alpha=1, ExternalForce="0,0,0") {
    angularDampingOverride=1
    angularSpringConstant=500
    bAngularSpring=true
    bOverrideAngularDamping=true
    bUseGravityOverride=true
    simulationSpace=RootRelative
    alphaBoolBlend {blendOption=Linear}
    boundBone {boneName=Hair_03_dyn_r}
    componentAppliedLinearAccClamp=<5000 5000 5000>
    componentLinearAccScale=<1 1 1>
    componentLinearVelScale=<1 1 1>
    gravityOverride {z=-980}
    physicsBodyDefinitions [{
      boundBone {boneName=Hair_03_dyn_r}
      boxExtents=<12 12 12>
      constraintSetup {angularLimitsMax=<0 5 8>, angularLimitsMin=<0 -5 -13>, angularTarget=<1 0 0>}
      localJointOffset=<1 0 0>
     }]
   }
   n13 = Anim Dynamics - Bone: Hair_04_dyn_r(ComponentPose=@n12, Alpha=1, ExternalForce="0,0,0") {
    angularDampingOverride=1
    angularSpringConstant=500
    bAngularSpring=true
    bOverrideAngularDamping=true
    bUseGravityOverride=true
    simulationSpace=RootRelative
    alphaBoolBlend {blendOption=Linear}
    boundBone {boneName=Hair_04_dyn_r}
    componentAppliedLinearAccClamp=<5000 5000 5000>
    componentLinearAccScale=<1 1 1>
    componentLinearVelScale=<1 1 1>
    gravityOverride {z=-980}
    physicsBodyDefinitions [{
      boundBone {boneName=Hair_04_dyn_r}
      boxExtents=<12 12 12>
      constraintSetup {angularLimitsMax=<0 10 13>, angularLimitsMin=<0 -10 -13>, angularTarget=<1 0 0>}
      localJointOffset=<1 0 0>
     }]
   }
   n14 = "Transform (Modify) Bone - Bone: ALT_corr_L"!(ComponentPose=@n13, bAlphaBoolEnabled:bool=True, Translation="0,0,-3.5", Rotation="-0.000,0.000,13.952", Scale="1,1,1") -> (Pose) {
    alphaInputType=Bool
    rotationMode=BMM_Additive
    translationMode=BMM_Additive
    boneToModify {boneName=ALT_corr_L}
    rotation {roll=13.952}
    translation {z=-3.5}
   }
   n15 = "Transform (Modify) Bone - Bone: ALT_corr_R"!(ComponentPose=@n14, bAlphaBoolEnabled:bool=True, Translation="0,0,-3", Rotation="-0.000,0.000,10.494", Scale="1,1,1") -> (Pose) {
    alphaInputType=Bool
    rotationMode=BMM_Additive
    translationMode=BMM_Additive
    boneToModify {boneName=ALT_corr_R}
    rotation {roll=10.494}
    translation {z=-3}
   }
   Output Pose(Result=Component To Local(ComponentPose=@n15)) {layerGroup=None}
   n16 = Control Rig!(Source=Copy Pose From Mesh() {bUseAttachedParent=true}, Alpha, SkyDive:bool=Get bIsSkyDiving()) -> (Pose) {controlRigAssetReference {blueprintRigClass=&/F_MED_CozyMechanic_Hen_FaceAcc_CtrlRig.F_MED_CozyMechanic_Hen_FaceAcc_CtrlRig_C}}
   Save cached pose 'cachPOse'(Pose=@n16) {cacheName=cachPOse}
  ]}
 EventGraph {nodes [
   Event BlueprintUpdateAnimation()
   > n1 = "Gravity Override State Function (Parameter Asset) (Nativized)"(
    CharacterPartAnimInstance=Self-Reference()
    GameTimeInSeconds=GetGameTimeInSeconds()
    Parameters=/BRCosmetics/Characters/Player/Male/Medium/Bodies/M_MED_IntenseCello/Meshes/M_MED_IntenseCello_Parameters
   )
   > Set Gravity Override(Gravity Override=@n1)
   n2 = Get bIsPlayingEmote()
   > Set Frontend(Frontend:bool=AND Boolean(A=Get bIsInFrontEnd(), B=NOT Boolean(A=@n2))) -> (Output_Get:bool)
   n3 = Get bRigidBodyEnabled()
   > Set AltskinFRend(AltskinFRend:bool=AND Boolean(A=Get bIsInFrontEnd(), B=NOT Boolean(A=@n3))) -> (Output_Get:bool)
  ]}
]
variables [
 AltskinFRend {type=boolean}
 Frontend {type=boolean}
 Gravity Override {type=Vector}
]
